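# Player anim layer Blueprint for the Beast Girl Storm Claws character, overriding Item and Locomotion layers, mostly passthroughs.
# ItemFullBodyOverride nests corrective additive, transform/landing and dive-ability state machines via cached poses, then applies a pelvis height offset.
# Transforms choose sequences on become-relevant; roar and transform can restrict to upper body; dive notifies update ability flags.
# Other layers add landing and idle additives and in-air landing anticipation into a fall loop.
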AnimBlueprint /BeastGirlStormClawsGameplay/Assets/Animation/Beast_Player_AnimBP
target_skeleton=&Fortnite_M_Avg_Player_Skeleton
graphs [
 AnimGraph {nodes [Output Pose()]}
 EventGraph {nodes [
   AnimNotify_DiveAbility_Landed() {is=title K2Node_Event}
   > Set bDiveAbilityLanded(bDiveAbilityLanded:bool=true) -> (Output_Get:bool)
   > Delay(Duration=1)
   > Set bDiveAbilityLanded(bDiveAbilityLanded=false)
   AnimNotify_DiveAttack_RoarComplete() {is=title K2Node_Event}
   > Set bDiveAbilityActive(bDiveAbilityActive:bool=false) -> (Output_Get:bool)
  ]}
 ItemAimOffset {nodes [
   Input Pose!() -> (Pose, AimOffsetAlpha:double, Yaw:double, Pitch:double) {name=InPose_AimOffset}
   Input Pose!() -> (Pose) {name=InPose_PreAimOffsets}
   Output Pose() {layerGroup=Item Override}
  ]}
 ItemFallAdditive {nodes [
   Input Pose() {name=InPoseItemFallAdditive}
   Output Pose() {layerGroup=Item Override}
  ]}
 ItemFallLandAdditive {nodes [
   Input Pose() {name=InPoseItemFallLandAdditive}
   n1 = "Sequence Player (additive) 'Beast_Land_CML_Additive'"() {bLoopAnimation=false, groupName=Landing, method=SyncGroup, sequence=&/Beast_Land_CML_Additive}
   n2 = Layered blend per bone(BasePose=Use cached pose 'InputPose'(), BlendPoses_0=@n1, BlendWeights_0=0.5) {bMeshSpaceRotationBlend=true, bRootSpaceRotationBlend=true, layerSetup [{branchFilters [{boneName=pelvis}]}]}
   n3 = "Sequence Player (additive) 'Beast_Land_CML_Additive'"() {bLoopAnimation=false, groupName=Landing, method=SyncGroup, sequence=&Beast_Land_CML_Additive}
   n4 = Layered blend per bone(BasePose=Use cached pose 'InputPose'(), BlendPoses_0=@n3, BlendWeights_0=0.9) {bMeshSpaceRotationBlend=true, bRootSpaceRotationBlend=true, layerSetup [{branchFilters [{boneName=root}]}]}
   n5 = Get bLandingAdditiveAllowLowerBody() -> (bLandingAdditiveAllowLowerBody:bool)
   n6 = Get bAllowLandingAdditive() -> (bAllowLandingAdditive:bool)
   Output Pose(Result=Blend Poses by bool(
     BlendPose_0=Use cached pose 'InputPose'()
     BlendPose_1=Blend Poses by bool(BlendPose_0=@n2, BlendPose_1=@n4, BlendTime_1=0.4, bActiveValue=@n5) {blendType=Linear}
     bActiveValue=NOT Boolean(A=@n6)
    ) {blendType=Linear}) {layerGroup=Item Override}
   Save cached pose 'InputPose'(Pose=Additive Identity Pose()) {cacheName=InputPose}
  ]}
 ItemFallLevel2Override {nodes [
   Input Pose() {name=InPose_FallLevel2}
   Output Pose() {layerGroup=Item Override}
  ]}
 ItemFinalPoseOverride {nodes [
   Input Pose() {name=InFinalPose}
   Output Pose() {layerGroup=Item Override}
  ]}
 ItemFullBodyAimOffsets {nodes [
   Input Pose!() -> (Pose, FullBody_AimOffsetAlpha:double, Aim_Yaw:double, Aim_Pitch:double) {name=InPoseFullBodyAimOffsets}
   Input Pose!() -> (Pose) {name=InPosePreFullBodyAimOffsets}
   Output Pose() {layerGroup=Item Override}
  ]}
 ItemFullBodyOverride {nodes [
   AnimGraphNode_StateMachine_1 = FullBody_StateMachine() {is=title AnimGraphNode_StateMachine}
   n1 = "Transform (Modify) Bone - Bone: pelvis"!(
    ComponentPose=Local To Component(LocalPose=@AnimGraphNode_StateMachine_1)
    Alpha=1
    Translation_X:double=0
    Translation_Y:double=0
    Translation_Z:double=Get HipOffsetVertical() -> (HipOffsetVertical:double)
    Rotation="0,0,0"
    Scale="1,1,1"
   ) -> (Pose) {translationMode=BMM_Additive, alphaBoolBlend {blendOption=Linear}, boneToModify {boneName=pelvis}, translation {z=-0.6}}
   Output Pose(Result=Component To Local(ComponentPose=@n1)) {layerGroup=Item Override}
   Save cached pose 'CorrectivePose'(Pose=Apply Additive(
     Base=Use cached pose 'InPose_FullBody'()
     Additive="Sequence Player (additive) 'Beast_nonTarget_CML_Additive'"() {bLoopAnimation=false, sequence=&/Beast_nonTarget_CML_Additive}
     Alpha=Get IdleCorrectiveAlpha() -> (IdleCorrectiveAlpha:double)
    ) {alphaBoolBlend {blendOption=Linear}}) {cacheName=CorrectivePose}
   AnimGraphNode_StateMachine_2 = TransformStates_w_Passthrough() {is=title AnimGraphNode_StateMachine, bAllowConduitEntryStates=true}
   Save cached pose 'TransformStates_w_PassthroughPose'(Pose=@AnimGraphNode_StateMachine_2) {cacheName=TransformStates_w_PassthroughPose}
   Save cached pose 'InPose_FullBody'(Pose=Input Pose() {name=InPoseFullBody}) {cacheName=InPose_FullBody, cachePoseName=InPose_FullBody}
   AnimGraphNode_StateMachine_0 = CorrectivePoseBlend() {is=title AnimGraphNode_StateMachine, bAllowConduitEntryStates=true}
   Save cached pose 'CorrectivePoseBlendedWithMotionMatching'(Pose=@AnimGraphNode_StateMachine_0) {cacheName=CorrectivePoseBlendedWithMotionMatching}
  ]}
 CorrectivePoseBlend {
  entry=Conduit
  path=ItemFullBodyOverride.AnimGraphNode_StateMachine_0.CorrectivePoseBlend
  ids=id
  nodes [
   AnimStateConduitNode_0 = Conduit()
   AnimStateNode_0 = Passthrough() {is=title AnimStateNode}
   AnimStateNode_1 = CorrectiveAdditive() {is=title AnimStateNode}
   AnimStateTransitionNode_0 = AnimStateTransitionNode() {priorityOrder=2}
   AnimStateTransitionNode_1 = AnimStateTransitionNode() {bDisabled=true, blendMode=CubicInOut, crossfadeDuration=0}
   AnimStateTransitionNode_2 = AnimStateTransitionNode() {blendMode=CubicInOut, crossfadeDuration=0.5, priorityOrder=2}
   AnimStateTransitionNode_3 = AnimStateTransitionNode() {blendMode=CubicInOut, crossfadeDuration=0.3}
   AnimStateTransitionNode_4 = AnimStateTransitionNode() {bDisabled=true, crossfadeDuration=0}
   AnimStateTransitionNode_5 = AnimStateTransitionNode() {crossfadeDuration=0.1}
   AnimStateTransitionNode_6 = AnimStateTransitionNode()
  ]
  links [
   {
    dst=CorrectiveAdditive
    rule=ItemFullBodyOverride.AnimGraphNode_StateMachine_0.CorrectivePoseBlend.AnimStateTransitionNode_2.AnimationTransitionGraph_0
    src=Conduit
    via=AnimStateTransitionNode_1
   }
   {
    dst=CorrectiveAdditive
    rule=ItemFullBodyOverride.AnimGraphNode_StateMachine_0.CorrectivePoseBlend.AnimStateTransitionNode_5.AnimationTransitionGraph_0
    src=Conduit
    via=AnimStateTransitionNode_4
   }
   {
    dst=CorrectiveAdditive
    rule=ItemFullBodyOverride.AnimGraphNode_StateMachine_0.CorrectivePoseBlend.AnimStateTransitionNode_6.Transition
    src=Conduit
    via=AnimStateTransitionNode_6
   }
   {
    dst=Passthrough
    rule=ItemFullBodyOverride.AnimGraphNode_StateMachine_0.CorrectivePoseBlend.AnimStateTransitionNode_0.AnimationTransitionGraph_1
    src=Conduit
    via=AnimStateTransitionNode_0
   }
   {
    dst=Passthrough
    rule=ItemFullBodyOverride.AnimGraphNode_StateMachine_0.CorrectivePoseBlend.AnimStateTransitionNode_3.AnimationTransitionGraph_1
    src=CorrectiveAdditive
    via=AnimStateTransitionNode_3
   }
   {
    dst=CorrectiveAdditive
    rule=ItemFullBodyOverride.AnimGraphNode_StateMachine_0.CorrectivePoseBlend.AnimStateTransitionNode_2.AnimationTransitionGraph_0
    src=Passthrough
    via=AnimStateTransitionNode_2
   }
   {
    dst=CorrectiveAdditive
    rule=ItemFullBodyOverride.AnimGraphNode_StateMachine_0.CorrectivePoseBlend.AnimStateTransitionNode_5.AnimationTransitionGraph_0
    src=Passthrough
    via=AnimStateTransitionNode_5
   }
  ]
 }
 Conduit {
  path=ItemFullBodyOverride.AnimGraphNode_StateMachine_0.CorrectivePoseBlend.AnimStateConduitNode_0.Conduit
  nodes [Result(bCanEnterTransition=true)]
 }
 Passthrough {
  path=ItemFullBodyOverride.AnimGraphNode_StateMachine_0.CorrectivePoseBlend.AnimStateNode_0.Passthrough
  nodes [Output Animation Pose(Result=Use cached pose 'InPose_FullBody'())]
 }
 CorrectiveAdditive {
  path=ItemFullBodyOverride.AnimGraphNode_StateMachine_0.CorrectivePoseBlend.AnimStateNode_1.CorrectiveAdditive
  nodes [Output Animation Pose(Result=Use cached pose 'CorrectivePose'())]
 }
 AnimationTransitionGraph_1 {
  path=ItemFullBodyOverride.AnimGraphNode_StateMachine_0.CorrectivePoseBlend.AnimStateTransitionNode_0.AnimationTransitionGraph_1
  nodes [Result(bCanEnterTransition=NOT Boolean(A=Get bSnapLowerBodyCorrective() -> (bSnapLowerBodyCorrective:bool)))]
 }
 AnimationTransitionGraph_0 {
  path=ItemFullBodyOverride.AnimGraphNode_StateMachine_0.CorrectivePoseBlend.AnimStateTransitionNode_2.AnimationTransitionGraph_0
  nodes [Result(bCanEnterTransition=Get bBlendLowerBodyCorrective() -> (bBlendLowerBodyCorrective:bool))]
 }
 AnimationTransitionGraph_1 {
  path=ItemFullBodyOverride.AnimGraphNode_StateMachine_0.CorrectivePoseBlend.AnimStateTransitionNode_3.AnimationTransitionGraph_1
  nodes [
   Result(bCanEnterTransition=Get bLowerBodyCorrective_BlendOff() -> (bLowerBodyCorrective_BlendOff:bool))
   NOT Boolean(A=Get bBlendLowerBodyCorrective())
  ]
 }
 AnimationTransitionGraph_0 {
  path=ItemFullBodyOverride.AnimGraphNode_StateMachine_0.CorrectivePoseBlend.AnimStateTransitionNode_5.AnimationTransitionGraph_0
  nodes [Result(bCanEnterTransition=Get bSnapLowerBodyCorrective())]
 }
 Transition {
  path=ItemFullBodyOverride.AnimGraphNode_StateMachine_0.CorrectivePoseBlend.AnimStateTransitionNode_6.Transition
  nodes [Result(bCanEnterTransition=true)]
 }
 FullBody_StateMachine {
  entry=Passthrough
  path=ItemFullBodyOverride.AnimGraphNode_StateMachine_1.FullBody_StateMachine
  ids=id
  nodes [
   AnimStateNode_0 = DiveAbility() {is=title AnimStateNode}
   AnimStateNode_1 = Passthrough()
   AnimStateTransitionNode_0 = AnimStateTransitionNode()
   AnimStateTransitionNode_9 = AnimStateTransitionNode()
  ]
  links [
   {
    dst=Passthrough
    rule=ItemFullBodyOverride.AnimGraphNode_StateMachine_1.FullBody_StateMachine.AnimStateTransitionNode_0.Transition
    src=DiveAbility
    via=AnimStateTransitionNode_0
   }
   {
    dst=DiveAbility
    rule=ItemFullBodyOverride.AnimGraphNode_StateMachine_1.FullBody_StateMachine.AnimStateTransitionNode_9.Transition
    src=Passthrough
    via=AnimStateTransitionNode_9
   }
  ]
 }
 DiveAbility {
  path=ItemFullBodyOverride.AnimGraphNode_StateMachine_1.FullBody_StateMachine.AnimStateNode_0.DiveAbility
  nodes [
   AnimGraphNode_StateMachine_0 = DiveAbility_States() {is=title AnimGraphNode_StateMachine}
   n1 = Local To Component(LocalPose=@AnimGraphNode_StateMachine_0)
   Output Animation Pose(Result=Component To Local(ComponentPose="Transform (Modify) Bone - Bone: Root"(ComponentPose=@n1, Alpha=1, Translation="0,0,0", Rotation="0,0,0", Scale="1,1,1") {rotationMode=BMM_Additive, alphaBoolBlend {blendOption=Linear}, boneToModify {boneName=Root}}))
  ]
 }
 DiveAbility_States {
  entry=Passthrough
  path=ItemFullBodyOverride.AnimGraphNode_StateMachine_1.FullBody_StateMachine.AnimStateNode_0.DiveAbility.AnimGraphNode_StateMachine_0.DiveAbility_States
  ids=id
  nodes [
   AnimStateNode_0 = DiveIntro() {is=title AnimStateNode}
   AnimStateNode_2 = Passthrough() {
    stateEntered {guid=#1}
    stateFullyBlended {guid=#2}
    stateLeft {guid=#3}
   }
   AnimStateNode_3 = Roar() {is=title AnimStateNode}
   AnimStateNode_4 = Charging() {
    is=title AnimStateNode
    stateEntered {guid=#4}
    stateFullyBlended {guid=#5}
    stateLeft {guid=#6}
   }
   AnimStateNode_5 = DiveLoop() {
    is=title AnimStateNode
    stateEntered {guid=#7}
    stateFullyBlended {guid=#8}
    stateLeft {guid=#9}
   }
   AnimStateNode_8 = Landed() {is=title AnimStateNode}
   AnimStateTransitionNode_0 = AnimStateTransitionNode() {transitionEnd {guid=#10, notifyName=DiveAttack_RoarComplete}}
   AnimStateTransitionNode_1 = AnimStateTransitionNode() {
    bAutomaticRuleBasedOnSequencePlayerInState=true
    blendMode=CubicInOut
    crossfadeDuration=0.4
    transitionEnd {guid=#11, notifyName=DiveAttack_RoarComplete}
   }
   AnimStateTransitionNode_10 = AnimStateTransitionNode() {blendMode=CubicInOut}
   AnimStateTransitionNode_11 = AnimStateTransitionNode() {crossfadeDuration=0.1}
   AnimStateTransitionNode_13 = AnimStateTransitionNode() {crossfadeDuration=0, transitionEnd {guid=#12, notifyName=DiveAbility_Landed}}
   AnimStateTransitionNode_14 = AnimStateTransitionNode()
   AnimStateTransitionNode_15 = AnimStateTransitionNode()
   AnimStateTransitionNode_2 = AnimStateTransitionNode() {
    bAutomaticRuleBasedOnSequencePlayerInState=true
    blendMode=CubicInOut
    blendProfileWrapper {blendProfile=&Fortnite_M_Avg_Player_Skeleton:SlowFeetFastUpper}
    transitionEnd {guid=#13, notifyName=DiveAttack_LandingComplete}
   }
   AnimStateTransitionNode_3 = AnimStateTransitionNode()
   AnimStateTransitionNode_4 = AnimStateTransitionNode()
   AnimStateTransitionNode_5 = AnimStateTransitionNode()
   AnimStateTransitionNode_6 = AnimStateTransitionNode() {transitionStart {guid=#14, notifyName=DiveAttack_LandingComplete}}
   AnimStateTransitionNode_8 = AnimStateTransitionNode()
   AnimStateTransitionNode_9 = AnimStateTransitionNode()
  ]
  links [
   {
    dst=DiveIntro
    rule=ItemFullBodyOverride.AnimGraphNode_StateMachine_1.FullBody_StateMachine.AnimStateNode_0.DiveAbility.AnimGraphNode_StateMachine_0.DiveAbility_States.AnimStateTransitionNode_10.AnimationTransitionGraph_0
    src=Charging
    via=AnimStateTransitionNode_10
   }
   {
    dst=Passthrough
    rule=ItemFullBodyOverride.AnimGraphNode_StateMachine_1.FullBody_StateMachine.AnimStateNode_0.DiveAbility.AnimGraphNode_StateMachine_0.DiveAbility_States.AnimStateTransitionNode_9.AnimationTransitionGraph_1
    src=Charging
    via=AnimStateTransitionNode_8
   }
   {
    dst=DiveLoop
    rule=ItemFullBodyOverride.AnimGraphNode_StateMachine_1.FullBody_StateMachine.AnimStateNode_0.DiveAbility.AnimGraphNode_StateMachine_0.DiveAbility_States.AnimStateTransitionNode_11.AnimationTransitionGraph_0
    src=DiveIntro
    via=AnimStateTransitionNode_11
   }
   {
    dst=Landed
    rule=ItemFullBodyOverride.AnimGraphNode_StateMachine_1.FullBody_StateMachine.AnimStateNode_0.DiveAbility.AnimGraphNode_StateMachine_0.DiveAbility_States.AnimStateTransitionNode_17.Transition
    src=DiveIntro
    via=AnimStateTransitionNode_14
   }
   {
    dst=Landed
    rule=ItemFullBodyOverride.AnimGraphNode_StateMachine_1.FullBody_StateMachine.AnimStateNode_0.DiveAbility.AnimGraphNode_StateMachine_0.DiveAbility_States.AnimStateTransitionNode_15.AnimationTransitionGraph_1
    src=DiveIntro
    via=AnimStateTransitionNode_15
   }
   {
    dst=Landed
    rule=ItemFullBodyOverride.AnimGraphNode_StateMachine_1.FullBody_StateMachine.AnimStateNode_0.DiveAbility.AnimGraphNode_StateMachine_0.DiveAbility_States.AnimStateTransitionNode_15.AnimationTransitionGraph_1
    src=DiveLoop
    via=AnimStateTransitionNode_13
   }
   {
    dst=Passthrough
    rule=ItemFullBodyOverride.AnimGraphNode_StateMachine_1.FullBody_StateMachine.AnimStateNode_0.DiveAbility.AnimGraphNode_StateMachine_0.DiveAbility_States.AnimStateTransitionNode_4.Transition
    src=DiveLoop
    via=AnimStateTransitionNode_4
   }
   {
    dst=Passthrough
    rule=ItemFullBodyOverride.AnimGraphNode_StateMachine_1.FullBody_StateMachine.AnimStateNode_0.DiveAbility.AnimGraphNode_StateMachine_0.DiveAbility_States.AnimStateTransitionNode_9.AnimationTransitionGraph_1
    src=DiveLoop
    via=AnimStateTransitionNode_9
   }
   {
    dst=Passthrough
    rule=ItemFullBodyOverride.AnimGraphNode_StateMachine_1.FullBody_StateMachine.AnimStateNode_0.DiveAbility.AnimGraphNode_StateMachine_0.DiveAbility_States.AnimStateTransitionNode_6.AnimationTransitionGraph_0
    src=Landed
    via=AnimStateTransitionNode_6
   }
   {
    dst=Roar
    rule=ItemFullBodyOverride.AnimGraphNode_StateMachine_1.FullBody_StateMachine.AnimStateNode_0.DiveAbility.AnimGraphNode_StateMachine_0.DiveAbility_States.AnimStateTransitionNode_2.AnimationTransitionGraph_0
    src=Landed
    via=AnimStateTransitionNode_2
   }
   {
    dst=Charging
    rule=ItemFullBodyOverride.AnimGraphNode_StateMachine_1.FullBody_StateMachine.AnimStateNode_0.DiveAbility.AnimGraphNode_StateMachine_0.DiveAbility_States.AnimStateTransitionNode_3.AnimationTransitionGraph_0
    src=Passthrough
    via=AnimStateTransitionNode_3
   }
   {
    dst=DiveLoop
    rule=ItemFullBodyOverride.AnimGraphNode_StateMachine_1.FullBody_StateMachine.AnimStateNode_0.DiveAbility.AnimGraphNode_StateMachine_0.DiveAbility_States.AnimStateTransitionNode_5.AnimationTransitionGraph_0
    src=Passthrough
    via=AnimStateTransitionNode_5
   }
   {
    dst=Passthrough
    rule=ItemFullBodyOverride.AnimGraphNode_StateMachine_1.FullBody_StateMachine.AnimStateNode_0.DiveAbility.AnimGraphNode_StateMachine_0.DiveAbility_States.AnimStateTransitionNode_0.AnimationTransitionGraph_0
    src=Roar
    via=AnimStateTransitionNode_0
   }
   {
    dst=Passthrough
    rule=ItemFullBodyOverride.AnimGraphNode_StateMachine_1.FullBody_StateMachine.AnimStateNode_0.DiveAbility.AnimGraphNode_StateMachine_0.DiveAbility_States.AnimStateTransitionNode_1.AnimationTransitionGraph_0
    src=Roar
    via=AnimStateTransitionNode_1
   }
  ]
 }
 DiveIntro {
  path=ItemFullBodyOverride.AnimGraphNode_StateMachine_1.FullBody_StateMachine.AnimStateNode_0.DiveAbility.AnimGraphNode_StateMachine_0.DiveAbility_States.AnimStateNode_0.DiveIntro
  nodes [Output Animation Pose(Result=Sequence Player 'BeastLeap_InAir_Attack_ExitToDive_CML'() {bLoopAnimation=false, sequence=&/BeastLeap_InAir_Attack_ExitToDive_CML})]
 }
 Passthrough {
  path=ItemFullBodyOverride.AnimGraphNode_StateMachine_1.FullBody_StateMachine.AnimStateNode_0.DiveAbility.AnimGraphNode_StateMachine_0.DiveAbility_States.AnimStateNode_2.Passthrough
  nodes [Output Animation Pose(Result=Use cached pose 'CorrectivePose'())]
 }
 Roar {
  path=ItemFullBodyOverride.AnimGraphNode_StateMachine_1.FullBody_StateMachine.AnimStateNode_0.DiveAbility.AnimGraphNode_StateMachine_0.DiveAbility_States.AnimStateNode_3.Roar
  nodes [
   n1 = Sequence Player 'BeastGirl_roar_CLM'() {bLoopAnimation=false, groupName=RoarSync, method=SyncGroup, sequence=&/BeastGirl_roar_CLM}
   n2 = Layered blend per bone(BasePose=Use cached pose 'CorrectivePose'(), BlendPoses_0=@n1) {bMeshSpaceRotationBlend=true, bRootSpaceRotationBlend=true, layerSetup [{branchFilters [{boneName=spine_01}]}]}
   n3 = Sequence Player 'BeastGirl_roar_CLM'() {bLoopAnimation=false, groupName=RoarSync, method=SyncGroup, sequence=&BeastGirl_roar_CLM}
   Output Animation Pose(Result=Blend Poses by bool(BlendPose_0=@n2, BlendPose_1=@n3, bActiveValue=Get bDiveState_Roar_AllowLowerBody() -> (bDiveState_Roar_AllowLowerBody:bool)) {blendType=Linear})
  ]
 }
 Charging {
  path=ItemFullBodyOverride.AnimGraphNode_StateMachine_1.FullBody_StateMachine.AnimStateNode_0.DiveAbility.AnimGraphNode_StateMachine_0.DiveAbility_States.AnimStateNode_4.Charging
  nodes [Output Animation Pose(Result=Sequence Player 'BeastLeap_InAir_Attack_Start_CML'() {bLoopAnimation=false, sequence=&/BeastLeap_InAir_Attack_Start_CML})]
 }
 DiveLoop {
  path=ItemFullBodyOverride.AnimGraphNode_StateMachine_1.FullBody_StateMachine.AnimStateNode_0.DiveAbility.AnimGraphNode_StateMachine_0.DiveAbility_States.AnimStateNode_5.DiveLoop
  nodes [Output Animation Pose(Result=Sequence Player 'BeastLeap_InAir_Attack_DiveLoop_CML'() {sequence=&/BeastLeap_InAir_Attack_DiveLoop_CML})]
 }
 Landed {
  path=ItemFullBodyOverride.AnimGraphNode_StateMachine_1.FullBody_StateMachine.AnimStateNode_0.DiveAbility.AnimGraphNode_StateMachine_0.DiveAbility_States.AnimStateNode_8.Landed
  nodes [Output Animation Pose(Result=Blend Poses by bool(
     BlendPose_0=Sequence Player 'Beast_InAir_Attack_End_Swimming_CLM'() {bLoopAnimation=false, sequence=&/Beast_InAir_Attack_End_Swimming_CLM}
     BlendPose_1=Sequence Player 'Beast_InAir_Attack_End_CML'() {bLoopAnimation=false, sequence=&/Beast_InAir_Attack_End_CML}
     bActiveValue=Get bIsInWater() -> (bIsInWater:bool)
    ) {blendType=Linear})]
 }
 AnimationTransitionGraph_0 {
  path=ItemFullBodyOverride.AnimGraphNode_StateMachine_1.FullBody_StateMachine.AnimStateNode_0.DiveAbility.AnimGraphNode_StateMachine_0.DiveAbility_States.AnimStateTransitionNode_0.AnimationTransitionGraph_0
  nodes [Result(bCanEnterTransition=Get bIsTacticalSprint() -> (bIsTacticalSprint:bool))]
 }
 AnimationTransitionGraph_0 {
  path=ItemFullBodyOverride.AnimGraphNode_StateMachine_1.FullBody_StateMachine.AnimStateNode_0.DiveAbility.AnimGraphNode_StateMachine_0.DiveAbility_States.AnimStateTransitionNode_1.AnimationTransitionGraph_0
  nodes [Result()]
 }
 AnimationTransitionGraph_0 {
  path=ItemFullBodyOverride.AnimGraphNode_StateMachine_1.FullBody_StateMachine.AnimStateNode_0.DiveAbility.AnimGraphNode_StateMachine_0.DiveAbility_States.AnimStateTransitionNode_10.AnimationTransitionGraph_0
  nodes [Result(bCanEnterTransition=Get bDiveState_Transition() -> (bDiveState_Transition:bool))]
 }
 AnimationTransitionGraph_0 {
  path=ItemFullBodyOverride.AnimGraphNode_StateMachine_1.FullBody_StateMachine.AnimStateNode_0.DiveAbility.AnimGraphNode_StateMachine_0.DiveAbility_States.AnimStateTransitionNode_11.AnimationTransitionGraph_0
  nodes [Result(bCanEnterTransition=Get bDiveState_Loop() -> (bDiveState_Loop:bool))]
 }
 AnimationTransitionGraph_1 {
  path=ItemFullBodyOverride.AnimGraphNode_StateMachine_1.FullBody_StateMachine.AnimStateNode_0.DiveAbility.AnimGraphNode_StateMachine_0.DiveAbility_States.AnimStateTransitionNode_15.AnimationTransitionGraph_1
  nodes [Result(bCanEnterTransition=Get bDiveState_Land() -> (bDiveState_Land:bool))]
 }
 Transition {
  path=ItemFullBodyOverride.AnimGraphNode_StateMachine_1.FullBody_StateMachine.AnimStateNode_0.DiveAbility.AnimGraphNode_StateMachine_0.DiveAbility_States.AnimStateTransitionNode_17.Transition
  nodes [Result(bCanEnterTransition=Get bIsInWater())]
 }
 AnimationTransitionGraph_0 {
  path=ItemFullBodyOverride.AnimGraphNode_StateMachine_1.FullBody_StateMachine.AnimStateNode_0.DiveAbility.AnimGraphNode_StateMachine_0.DiveAbility_States.AnimStateTransitionNode_2.AnimationTransitionGraph_0
  nodes [Result()]
 }
 AnimationTransitionGraph_0 {
  path=ItemFullBodyOverride.AnimGraphNode_StateMachine_1.FullBody_StateMachine.AnimStateNode_0.DiveAbility.AnimGraphNode_StateMachine_0.DiveAbility_States.AnimStateTransitionNode_3.AnimationTransitionGraph_0
  nodes [Result(bCanEnterTransition=Get bDiveState_EnterChargeState() -> (bDiveState_EnterChargeState:bool))]
 }
 Transition {
  path=ItemFullBodyOverride.AnimGraphNode_StateMachine_1.FullBody_StateMachine.AnimStateNode_0.DiveAbility.AnimGraphNode_StateMachine_0.DiveAbility_States.AnimStateTransitionNode_4.Transition
  nodes [Result(bCanEnterTransition=Get bIsInWater())]
 }
 AnimationTransitionGraph_0 {
  path=ItemFullBodyOverride.AnimGraphNode_StateMachine_1.FullBody_StateMachine.AnimStateNode_0.DiveAbility.AnimGraphNode_StateMachine_0.DiveAbility_States.AnimStateTransitionNode_5.AnimationTransitionGraph_0
  nodes [Result(bCanEnterTransition=Get bDiveState_Loop())]
 }
 AnimationTransitionGraph_0 {
  path=ItemFullBodyOverride.AnimGraphNode_StateMachine_1.FullBody_StateMachine.AnimStateNode_0.DiveAbility.AnimGraphNode_StateMachine_0.DiveAbility_States.AnimStateTransitionNode_6.AnimationTransitionGraph_0
  nodes [Result(bCanEnterTransition=Get bIsTacticalSprint())]
 }
 AnimationTransitionGraph_1 {
  path=ItemFullBodyOverride.AnimGraphNode_StateMachine_1.FullBody_StateMachine.AnimStateNode_0.DiveAbility.AnimGraphNode_StateMachine_0.DiveAbility_States.AnimStateTransitionNode_9.AnimationTransitionGraph_1
  nodes [Result(bCanEnterTransition=Get bDiveState_Inactive() -> (bDiveState_Inactive:bool))]
 }
 Passthrough {
  path=ItemFullBodyOverride.AnimGraphNode_StateMachine_1.FullBody_StateMachine.AnimStateNode_1.Passthrough
  nodes [
   n1 = Use cached pose 'TransformStates_w_PassthroughPose'()
   Output Animation Pose(Result=Blend Poses by bool(
     BlendPose_0=Slot 'TDiA.Custom_01'(Source=@n1) {slotName=TDiA.Custom_01}
     BlendPose_1=Use cached pose 'TransformStates_w_PassthroughPose'()
     BlendTime_0=0.05
     BlendTime_1=0.2
     bActiveValue=Get bPlayJumpStart() -> (bPlayJumpStart:bool)
    ) {blendType=Linear})
  ]
 }
 Transition {
  path=ItemFullBodyOverride.AnimGraphNode_StateMachine_1.FullBody_StateMachine.AnimStateTransitionNode_0.Transition
  nodes [Result(bCanEnterTransition=NOT Boolean(A=Get bDiveAbilityActive() -> (bDiveAbilityActive:bool)))]
 }
 Transition {
  path=ItemFullBodyOverride.AnimGraphNode_StateMachine_1.FullBody_StateMachine.AnimStateTransitionNode_9.Transition
  nodes [Result(bCanEnterTransition=Get bDiveAbilityActive())]
 }
 TransformStates_w_Passthrough {
  entry=Passthrough
  path=ItemFullBodyOverride.AnimGraphNode_StateMachine_2.TransformStates_w_Passthrough
  ids=id
  nodes [
   AnimStateNode_0 = Passthrough() {
    stateEntered {guid=#15}
    stateFullyBlended {guid=#16}
    stateLeft {guid=#17}
   }
   AnimStateNode_1 = Transform() {
    is=title AnimStateNode
    stateEntered {guid=#18}
    stateFullyBlended {guid=#19}
    stateLeft {guid=#20}
   }
   AnimStateNode_2 = LandingOverride() {is=title AnimStateNode}
   AnimStateTransitionNode_2 = AnimStateTransitionNode() {bAutomaticRuleBasedOnSequencePlayerInState=true, blendMode=CubicInOut, crossfadeDuration=0.3}
   AnimStateTransitionNode_3 = AnimStateTransitionNode()
   AnimStateTransitionNode_4 = AnimStateTransitionNode()
   AnimStateTransitionNode_5 = AnimStateTransitionNode() {blendMode=CircularOut, crossfadeDuration=0.3}
   AnimStateTransitionNode_6 = AnimStateTransitionNode() {automaticRuleTriggerTime=0.3, bAutomaticRuleBasedOnSequencePlayerInState=true, blendMode=CircularOut, crossfadeDuration=0.3}
   AnimStateTransitionNode_7 = AnimStateTransitionNode()
   AnimStateTransitionNode_8 = AnimStateTransitionNode() {bDisabled=true}
  ]
  links [
   {
    dst=Passthrough
    rule=ItemFullBodyOverride.AnimGraphNode_StateMachine_2.TransformStates_w_Passthrough.AnimStateTransitionNode_5.Transition
    src=LandingOverride
    via=AnimStateTransitionNode_5
   }
   {
    dst=Passthrough
    rule=ItemFullBodyOverride.AnimGraphNode_StateMachine_2.TransformStates_w_Passthrough.AnimStateTransitionNode_6.Transition
    src=LandingOverride
    via=AnimStateTransitionNode_6
   }
   {
    dst=Passthrough
    rule=ItemFullBodyOverride.AnimGraphNode_StateMachine_2.TransformStates_w_Passthrough.AnimStateTransitionNode_8.Transition
    src=LandingOverride
    via=AnimStateTransitionNode_8
   }
   {
    dst=Transform
    rule=ItemFullBodyOverride.AnimGraphNode_StateMachine_2.TransformStates_w_Passthrough.AnimStateTransitionNode_1.Transition
    src=LandingOverride
    via=AnimStateTransitionNode_7
   }
   {
    dst=LandingOverride
    rule=ItemFullBodyOverride.AnimGraphNode_StateMachine_2.TransformStates_w_Passthrough.AnimStateTransitionNode_4.Transition
    src=Passthrough
    via=AnimStateTransitionNode_4
   }
   {
    dst=Transform
    rule=ItemFullBodyOverride.AnimGraphNode_StateMachine_2.TransformStates_w_Passthrough.AnimStateTransitionNode_1.Transition
    src=Passthrough
    via=AnimStateTransitionNode_3
   }
   {
    dst=Passthrough
    rule=ItemFullBodyOverride.AnimGraphNode_StateMachine_2.TransformStates_w_Passthrough.AnimStateTransitionNode_2.Transition
    src=Transform
    via=AnimStateTransitionNode_2
   }
  ]
 }
 Passthrough {
  path=ItemFullBodyOverride.AnimGraphNode_StateMachine_2.TransformStates_w_Passthrough.AnimStateNode_0.Passthrough
  nodes [Output Animation Pose(Result=Use cached pose 'CorrectivePoseBlendedWithMotionMatching'())]
 }
 Transform {
  path=ItemFullBodyOverride.AnimGraphNode_StateMachine_2.TransformStates_w_Passthrough.AnimStateNode_1.Transform
  nodes [
   n1 = Sequence Player() {
    becomeRelevantFunction {bSelfContext=true, memberGuid=#21, memberName=Transform_SetupTransformSequence}
    node {bLoopAnimation=false, groupName=TransformSync, method=SyncGroup, becomeRelevantFunction {functionName=OnBecomeRelevant}}
   }
   n2 = Layered blend per bone(BasePose=Use cached pose 'CorrectivePoseBlendedWithMotionMatching'(), BlendPoses_0=@n1) {bMeshSpaceRotationBlend=true, bRootSpaceRotationBlend=true, layerSetup [{branchFilters [{boneName=spine_01}]}]}
   n3 = Sequence Player() {
    becomeRelevantFunction {bSelfContext=true, memberGuid=#21, memberName=Transform_SetupTransformSequence}
    node {bLoopAnimation=false, groupName=TransformSync, method=SyncGroup, becomeRelevantFunction {functionName=OnBecomeRelevant}}
   }
   Output Animation Pose(Result=Blend Poses by bool(BlendPose_0=@n2, BlendPose_1=@n3, bActiveValue=Get bTransformAllowLowerBody() -> (bTransformAllowLowerBody:bool)) {blendType=Linear})
  ]
 }
 LandingOverride {
  path=ItemFullBodyOverride.AnimGraphNode_StateMachine_2.TransformStates_w_Passthrough.AnimStateNode_2.LandingOverride
  nodes [Output Animation Pose(Result=Sequence Player 'Beast_Land_CML'() {bLoopAnimation=false, sequence=&/Beast_Land_CML})]
 }
 Transition {
  path=ItemFullBodyOverride.AnimGraphNode_StateMachine_2.TransformStates_w_Passthrough.AnimStateTransitionNode_1.Transition
  nodes [Result(bCanEnterTransition=Get bEnterTransformState() -> (bEnterTransformState:bool))]
 }
 Transition {path=ItemFullBodyOverride.AnimGraphNode_StateMachine_2.TransformStates_w_Passthrough.AnimStateTransitionNode_2.Transition, nodes [Result()]}
 Transition {
  path=ItemFullBodyOverride.AnimGraphNode_StateMachine_2.TransformStates_w_Passthrough.AnimStateTransitionNode_4.Transition
  nodes [Result(bCanEnterTransition=Get bAllowLowerbodyLandingOverride() -> (bAllowLowerbodyLandingOverride:bool))]
 }
 Transition {
  path=ItemFullBodyOverride.AnimGraphNode_StateMachine_2.TransformStates_w_Passthrough.AnimStateTransitionNode_5.Transition
  nodes [Result(bCanEnterTransition=NOT Boolean(A=Get bAllowLowerbodyLandingOverride()))]
 }
 Transition {path=ItemFullBodyOverride.AnimGraphNode_StateMachine_2.TransformStates_w_Passthrough.AnimStateTransitionNode_6.Transition, nodes [Result()]}
 Transition {
  path=ItemFullBodyOverride.AnimGraphNode_StateMachine_2.TransformStates_w_Passthrough.AnimStateTransitionNode_8.Transition
  nodes [Result(bCanEnterTransition=Get bIsMovingOnFoot() -> (bIsMovingOnFoot:bool))]
 }
 ItemIdleOverride {nodes [
   Input Pose() {name=InPose_Idle}
   Output Pose() {layerGroup=Item Override}
  ]}
 ItemJetPackJumpAdditive {nodes [Output Pose() {layerGroup=Item Override}]}
 ItemJetPackStartAdditive {nodes [Output Pose() {layerGroup=Item Override}]}
 ItemJumpLoopAdditive {nodes [
   Input Pose() {name=InPoseItemJumpLoopAdditive}
   Output Pose() {layerGroup=Item Override}
  ]}
 ItemJumpOverride {nodes [
   Input Pose() {name=InPoseJump}
   Output Pose() {layerGroup=Item Override}
  ]}
 ItemJumpUpAdditive {nodes [
   Input Pose() {name=InPoseItemJumpUpAdditive}
   Output Pose() {layerGroup=Item Override}
  ]}
 ItemPreAimOffsets {nodes [
   Input Pose() {name=InPosePreAimOffsets}
   Output Pose() {layerGroup=Item Override}
  ]}
 ItemPreIK {nodes [
   Input Pose() {name=InPose_PreIK}
   Output Pose() {layerGroup=Item Override}
  ]}
 ItemRightHandIK {nodes [
   Input Pose() {name=InPosePreRightHandIK}
   Input Pose() {name=InPosePostRightHandIK}
   Output Pose() {layerGroup=Item Override}
  ]}
 ItemSkeletalControl {nodes [
   Input Pose()
   Output Pose() {layerGroup=Item Override}
  ]}
 ItemSwimJumpFallLoop {nodes [Output Pose() {layerGroup=Item Override}]}
 ItemSwimJumpSurfaceEnd {nodes [Output Pose() {layerGroup=Item Override}]}
 ItemSwimJumpSurfaceLoop {nodes [Output Pose() {layerGroup=Item Override}]}
 ItemUpperBody {nodes [
   Input Pose!() -> (Pose, InputParam:Fort Anim Input Adjusted Aim Structure) {name=InPoseUpperBody}
   Output Pose() {layerGroup=Item Override}
  ]}
 ItemZipLineStartAdditive {nodes [Output Pose() {layerGroup=Item Override}]}
 LocomotionLayer_Additives {nodes [Output Pose() {layerGroup=LocomotionOverride}]}
 LocomotionLayer_BaseLocomotion {nodes [
   Input Pose!() -> (Pose) {name=InPose_BaseLocomotion}
   Output Pose() {layerGroup=LocomotionOverride}
  ]}
 LocomotionLayer_FinalPose {nodes [
   Input Pose() {name=InPoseFinalPose}
   Output Pose() {layerGroup=LocomotionOverride}
  ]}
 LocomotionLayer_FullBody {nodes [
   Input Pose() {name=InPose_FullBody}
   Output Pose() {layerGroup=LocomotionOverride}
  ]}
 LocomotionLayer_FullLocomotion {nodes [
   n1 = "Sequence Player (additive) 'Beast_IdleCycle_Additive'"() {sequence=/BeastGirlStormClawsGameplay/Assets/Beast_IdleCycle_Additive}
   n2 = Use cached pose 'InPose_FullLocomotion'()
   n3 = Get IdleCorrectiveAlpha()
   Output Pose(Result=Blend Poses by bool(
     BlendPose_0=Apply Additive(Base=@n2, Additive=@n1, Alpha=@n3) {alphaBoolBlend {blendOption=Linear}}
     BlendPose_1=Use cached pose 'InPose_FullLocomotion'()
     BlendTime_0=0.5
     BlendTime_1=0.2
     bActiveValue=Get bAllowAdditiveIdleLoop() -> (bAllowAdditiveIdleLoop:bool)
    ) {blendType=CircularOut}) {layerGroup=LocomotionOverride}
   Save cached pose 'InPose_FullLocomotion'(Pose=Input Pose() {name=InPose_FullLocomotion}) {cacheName=InPose_FullLocomotion}
  ]}
 LocomotionLayer_GroundLandAdditive {nodes [
   Input Pose()
   Output Pose() {layerGroup=LocomotionOverride}
  ]}
 LocomotionLayer_InAir {nodes [
   Input Pose() {name=InPose_InAirSourcePose}
   AnimGraphNode_StateMachine_0 = InAir() {is=title AnimGraphNode_StateMachine}
   Output Pose(Result=Modify Curve(SourcePose=@AnimGraphNode_StateMachine_0, CurveValues_0=1, Alpha=1) {curveNames [DisableAdditiveProceduralLayering], curveValues [0]}) {layerGroup=LocomotionOverride}
   Save cached pose 'InputPoseInAir'(Pose=Input Pose() {name=InPose_InAir}) {cacheName=InputPoseInAir}
  ]}
 InAir {
  entry=InAir
  path=LocomotionLayer_InAir.AnimGraphNode_StateMachine_0.InAir
  ids=id
  nodes [
   AnimStateNode_0 = InAir() {is=title AnimStateNode}
   AnimStateNode_1 = LandingAnticipation() {is=title AnimStateNode}
   AnimStateNode_2 = FallLoop() {is=title AnimStateNode}
   AnimStateTransitionNode_0 = AnimStateTransitionNode() {blendMode=CubicInOut, crossfadeDuration=0.5}
   AnimStateTransitionNode_1 = AnimStateTransitionNode() {bAutomaticRuleBasedOnSequencePlayerInState=true}
  ]
  links [
   {
    dst=LandingAnticipation
    rule=LocomotionLayer_InAir.AnimGraphNode_StateMachine_0.InAir.AnimStateTransitionNode_0.Transition
    src=InAir
    via=AnimStateTransitionNode_0
   }
   {
    dst=FallLoop
    rule=LocomotionLayer_InAir.AnimGraphNode_StateMachine_0.InAir.AnimStateTransitionNode_1.Transition
    src=LandingAnticipation
    via=AnimStateTransitionNode_1
   }
  ]
 }
 InAir {
  path=LocomotionLayer_InAir.AnimGraphNode_StateMachine_0.InAir.AnimStateNode_0.InAir
  nodes [Output Animation Pose(Result=Slot 'TDiA.Custom_01'(Source=Use cached pose 'InputPoseInAir'()) {slotName=TDiA.Custom_01})]
 }
 LandingAnticipation {
  path=LocomotionLayer_InAir.AnimGraphNode_StateMachine_0.InAir.AnimStateNode_1.LandingAnticipation
  nodes [Output Animation Pose(Result=Sequence Player 'Beast_Land_Anticipation_CML'() {bLoopAnimation=false, sequence=&/Beast_Land_Anticipation_CML})]
 }
 FallLoop {
  path=LocomotionLayer_InAir.AnimGraphNode_StateMachine_0.InAir.AnimStateNode_2.FallLoop
  nodes [Output Animation Pose(Result=Sequence Player 'Beast_Jump_FallLoop'() {sequence=&/Beast_Jump_FallLoop})]
 }
 Transition {
  path=LocomotionLayer_InAir.AnimGraphNode_StateMachine_0.InAir.AnimStateTransitionNode_0.Transition
  nodes [Result(bCanEnterTransition=Get bIsFalling())]
 }
 Transition {path=LocomotionLayer_InAir.AnimGraphNode_StateMachine_0.InAir.AnimStateTransitionNode_1.Transition, nodes [Result()]}
 LocomotionLayer_LowerBody {nodes [
   Input Pose() {name=InPose_LowerBody}
   Output Pose() {layerGroup=LocomotionOverride}
  ]}
 LocomotionLayer_UpperBody {nodes [
   Input Pose() {name=InPose_UpperBody}
   Output Pose() {layerGroup=LocomotionOverride}
  ]}
 Transform_SetupTransformSequence {nodes [
   n1 = Transform_SetupTransformSequence!() -> (then, Context:"Anim Update Context Structure (by ref)", Node:"Anim Node Reference Structure (by ref)") {is=title K2Node_FunctionEntry}
   n2 = Select(
    Option 0:Anim Sequence Object Reference=&/BeastGirl_transformBack_Start_CLM
    Option 1:Anim Sequence Object Reference=&/BeastGirl_transformBack__InWater_Start_CLM
    Index=Get bIsInWater()
   ) -> (ReturnValue:Anim Sequence Object Reference)
   n3 = Get bTransformExitManual() -> (bTransformExitManual:bool)
   n4 = Get bIsInWater()
   > SetSequenceWithInertialBlending!(
    execute
    UpdateContext:"Anim Update Context Structure (by ref)"=@n1
    SequencePlayer:"Sequence Player Reference Structure (by ref)"=Convert to Sequence Player!(Node:"Anim Node Reference Structure (by ref)"=@n1.Node) -> (SequencePlayer:Sequence Player Reference Structure, Result:bool)
    Sequence:Anim Sequence Base Object Reference=Select(
     Option 0=Select(
      Option 0=&/CheerTransformPart2_CML
      Option 1=&/CheerTransformPart2_InWater_CML_
      Index=@n4
     )
     Option 1=Select(Option 0=@n2, Option 1=&/BeastGirl_TransformBack_Fast_Start_CLM, Index=@n3)
     Index=Get bTransformExitActive() -> (bTransformExitActive:bool)
    )
    BlendTime:float=0.2
   ) -> (then, ReturnValue:Sequence Player Reference Structure)
   > Return Node()
  ]}
]
variables []
