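# Player animation layer Blueprint for the Fortnite average male skeleton that plays the Beast Girl transform-back ending animation.
# Only the ItemFullBodyOverride layer has logic: a state machine leaves cached-pose passthrough when DoTransformBackFromBeast is true, then returns automatically as the sequence ends.
# On becoming relevant, SetupTransformBackAnim picks a normal, fast or water sequence from the exit and water flags, using inertial blending.
# The other item layers are unwired stubs, and nothing here sets the flags.
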AnimBlueprint /BeastGirlStormClawsGameplay/Assets/Animation/Beast_Player_Avg_AnimBP
target_skeleton=&Fortnite_M_Avg_Player_Skeleton
graphs [
 AnimGraph {nodes [Output Pose()]}
 EventGraph {nodes [
   TryGetPawnOwner()
   Event BlueprintUpdateAnimation()
  ]}
 ItemAimOffset {nodes [
   Input Pose!() -> (Pose, AimOffsetAlpha:double, Yaw:double, Pitch:double) {name=InPose_AimOffset}
   Input Pose!() -> (Pose) {name=InPose_PreAimOffsets}
   Output Pose() {layerGroup=Item Override}
  ]}
 ItemFallAdditive {nodes [
   Input Pose() {name=InPoseItemFallAdditive}
   Output Pose() {layerGroup=Item Override}
  ]}
 ItemFallLandAdditive {nodes [
   Input Pose() {name=InPoseItemFallLandAdditive}
   Output Pose() {layerGroup=Item Override}
  ]}
 ItemFallLevel2Override {nodes [
   Input Pose() {name=InPose_FallLevel2}
   Output Pose() {layerGroup=Item Override}
  ]}
 ItemFinalPoseOverride {nodes [
   Input Pose() {name=InFinalPose}
   Output Pose() {layerGroup=Item Override}
  ]}
 ItemFullBodyAimOffsets {nodes [
   Input Pose!() -> (Pose, FullBody_AimOffsetAlpha:double, Aim_Yaw:double, Aim_Pitch:double) {name=InPoseFullBodyAimOffsets}
   Input Pose!() -> (Pose) {name=InPosePreFullBodyAimOffsets}
   Output Pose() {layerGroup=Item Override}
  ]}
 ItemFullBodyOverride {nodes [
   AnimGraphNode_StateMachine_0 = Transform StateMachine() {is=title AnimGraphNode_StateMachine}
   Output Pose(Result=@AnimGraphNode_StateMachine_0) {layerGroup=Item Override}
   Save cached pose 'InPoseFullBody'(Pose=Input Pose() {name=InPoseFullBody}) {cacheName=InPoseFullBody}
  ]}
 Transform StateMachine {
  entry=Passthrough
  path=ItemFullBodyOverride.AnimGraphNode_StateMachine_0.Transform StateMachine
  ids=id
  nodes [
   AnimStateNode_0 = Passthrough() {is=title AnimStateNode}
   AnimStateNode_1 = Transform() {is=title AnimStateNode}
   AnimStateTransitionNode_0 = AnimStateTransitionNode()
   AnimStateTransitionNode_1 = AnimStateTransitionNode() {bAutomaticRuleBasedOnSequencePlayerInState=true}
  ]
  links [
   {
    dst=Transform
    rule=ItemFullBodyOverride.AnimGraphNode_StateMachine_0.Transform StateMachine.AnimStateTransitionNode_0.Transition
    src=Passthrough
    via=AnimStateTransitionNode_0
   }
   {
    dst=Passthrough
    rule=ItemFullBodyOverride.AnimGraphNode_StateMachine_0.Transform StateMachine.AnimStateTransitionNode_1.Transition
    src=Transform
    via=AnimStateTransitionNode_1
   }
  ]
 }
 Passthrough {
  path=ItemFullBodyOverride.AnimGraphNode_StateMachine_0.Transform StateMachine.AnimStateNode_0.Passthrough
  nodes [Output Animation Pose(Result=Use cached pose 'InPoseFullBody'())]
 }
 Transform {
  path=ItemFullBodyOverride.AnimGraphNode_StateMachine_0.Transform StateMachine.AnimStateNode_1.Transform
  nodes [
   n1 = Sequence Player() {
    becomeRelevantFunction {bSelfContext=true, memberGuid=#1, memberName=SetupTransformBackAnim}
    node {bLoopAnimation=false, becomeRelevantFunction {functionName=OnBecomeRelevant}}
   }
   Output Animation Pose(Result=@n1)
  ]
 }
 Transition {
  path=ItemFullBodyOverride.AnimGraphNode_StateMachine_0.Transform StateMachine.AnimStateTransitionNode_0.Transition
  nodes [Result(bCanEnterTransition=Get DoTransformBackFromBeast() -> (DoTransformBackFromBeast:bool))]
 }
 Transition {path=ItemFullBodyOverride.AnimGraphNode_StateMachine_0.Transform StateMachine.AnimStateTransitionNode_1.Transition, nodes [Result()]}
 ItemIdleOverride {nodes [
   Input Pose() {name=InPose_Idle}
   Output Pose() {layerGroup=Item Override}
  ]}
 ItemJetPackJumpAdditive {nodes [Output Pose() {layerGroup=Item Override}]}
 ItemJetPackStartAdditive {nodes [Output Pose() {layerGroup=Item Override}]}
 ItemJumpLoopAdditive {nodes [
   Input Pose() {name=InPoseItemJumpLoopAdditive}
   Output Pose() {layerGroup=Item Override}
  ]}
 ItemJumpOverride {nodes [
   Input Pose() {name=InPoseJump}
   Output Pose() {layerGroup=Item Override}
  ]}
 ItemJumpUpAdditive {nodes [
   Input Pose() {name=InPoseItemJumpUpAdditive}
   Output Pose() {layerGroup=Item Override}
  ]}
 ItemPreAimOffsets {nodes [
   Input Pose() {name=InPosePreAimOffsets}
   Output Pose() {layerGroup=Item Override}
  ]}
 ItemPreIK {nodes [
   Input Pose() {name=InPose_PreIK}
   Output Pose() {layerGroup=Item Override}
  ]}
 ItemRightHandIK {nodes [
   Input Pose() {name=InPosePreRightHandIK}
   Input Pose() {name=InPosePostRightHandIK}
   Output Pose() {layerGroup=Item Override}
  ]}
 ItemSkeletalControl {nodes [
   Input Pose()
   Output Pose() {layerGroup=Item Override}
  ]}
 ItemSwimJumpFallLoop {nodes [Output Pose() {layerGroup=Item Override}]}
 ItemSwimJumpSurfaceEnd {nodes [Output Pose() {layerGroup=Item Override}]}
 ItemSwimJumpSurfaceLoop {nodes [Output Pose() {layerGroup=Item Override}]}
 ItemUpperBody {nodes [
   Input Pose!() -> (Pose, InputParam:Fort Anim Input Adjusted Aim Structure) {name=InPoseUpperBody}
   Output Pose() {layerGroup=Item Override}
  ]}
 ItemZipLineStartAdditive {nodes [Output Pose() {layerGroup=Item Override}]}
 SetupTransformBackAnim {nodes [
   n1 = SetupTransformBackAnim!() -> (then, Context:"Anim Update Context Structure (by ref)", Node:"Anim Node Reference Structure (by ref)") {is=title K2Node_FunctionEntry}
   n2 = Select(
    Option 0:Anim Sequence Object Reference=&/BeastGirl_TransformBack_Fast_End_CFM
    Option 1:Anim Sequence Object Reference=&/BeastGirl_TransformBack_Fast_End_FromWater_CFM
    Index=Get bIsInWater() -> (bIsInWater:bool)
   ) -> (ReturnValue:Anim Sequence Object Reference)
   n3 = Get TransformExit_Manual() -> (TransformExit_Manual:bool)
   n4 = Get bIsInWater()
   > SetSequenceWithInertialBlending!(
    execute
    UpdateContext:"Anim Update Context Structure (by ref)"=@n1
    SequencePlayer:"Sequence Player Reference Structure (by ref)"=Convert to Sequence Player!(Node:"Anim Node Reference Structure (by ref)"=@n1.Node) -> (SequencePlayer:Sequence Player Reference Structure, Result:bool)
    Sequence:Anim Sequence Base Object Reference=Select(
     Option 0=Select(
      Option 0=&/BeastGirl_transformBack_End_CFM
      Option 1=&/BeastGirl_transformBack_InWater_End_CFM
      Index=@n4
     )
     Option 1=Select(Option 0=@n2, Option 1=&BeastGirl_TransformBack_Fast_End_CFM, Index=@n3)
     Index=Get TransformExit_Active() -> (TransformExit_Active:bool)
    )
    BlendTime:float=0.2
   ) -> (then, ReturnValue:Sequence Player Reference Structure)
   > Return Node()
  ]}
]
variables [
 bIsInWater {type=boolean}
 DoTransformBackFromBeast {type=boolean}
 TransformExit_Active {type=boolean}
 TransformExit_Manual {type=boolean}
]
