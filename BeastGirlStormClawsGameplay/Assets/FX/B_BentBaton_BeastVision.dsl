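; Niagara actor Blueprint for the BentBaton beast vision effect that fits its particle system to the screen's proportions.
; On BeginPlay it normalizes the viewport size, adds its Scale variable and writes the result to the Niagara "Ratio" Vector2 parameter.
; The value is set only once, with no update on later viewport resizes, and the construction script is empty.
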(event EventBeginPlay
  (Niagara|SetNiagaraVariable(Vector2) (Variables|NiagaraActor|GetNiagaraComponent) "Ratio" (+ (Math|Vector2D|Normalize2D (Viewport|GetViewportSize)) (Variables|Default|GetScale))))

(fn ConstructionScript)
