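# Unlit translucent Niagara sprite material for the BentBaton Beast Vision effect, drawn without depth testing as an edge vignette.
# An inverted square edge mask blends two colors, tinted by particle color and passed through HDR emissive adjustments.
# Opacity adds radially panned lightning noise to the mask, using pre-radial derivatives for mip selection, then applies contrast and particle alpha.
# Dynamic parameters drive it, though Square Mask Multiplier feeds both mask inputs, leaving Exponent unused.
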Material /BeastGirlStormClawsGameplay/Assets/FX/Materials/M_BentBaton_BeastVision
graph {
 properties {
  bAutomaticallySetUsageInEditor=false
  bDisableDepthTest=true
  bUsedWithNiagaraSprites=true
  blendMode=BLEND_Translucent
  shadingModel=MSM_Unlit
  twoSided=true
  usedShadingModels=Unlit
 }
 nodes [
  n1 = DynamicParameter() {paramNames [Square Mask Multiplier, Square Mask Exponent, Screen Contrast, Detail Multiplier]}
  n2 = OneMinus(None=MaterialFunctionCall!(UVs=TextureCoordinate(), Mult=@n1.Square Mask Multiplier, Exp=@n1.Square Mask Multiplier) {materialFunction=/Game/Effects/Fort_Effects/MaterialFunctions/MF_SquareEdgeMask}.SquareMask)
  n3 = ParticleColor()
  n4 = Multiply(
   A=LinearInterpolate(
    A=VectorParameter() {parameterName=Color A, defaultValue=<rgba 0.0677 0 0.0011 0>}.RGB
    B=VectorParameter() {parameterName=Color B, defaultValue=<rgba 1 0 0.0167 0>}.RGB
    Alpha=Power(Base=@n2) {constExponent=0.5}
   )
   B=@n3.RGB
  )
  MaterialFunctionCall(Emissive=MaterialFunctionCall(LDR Input=Saturate(None=@n4), HDR Tint=@n4) {materialFunction=&DeriveHDRfromLDR}.RemappedRange) {materialFunction=&MF_FX_EmissiveAdjustments}
  n5 = VectorParameter() {parameterName=Radial Noise UV, defaultValue=<rgba 0.5 0.5 0 0.125>}
  n6 = AppendVector(A=@n5.R, B=@n5.R)
  n7 = Subtract(A=Multiply(A=TextureCoordinate(), B=@n6), B=Multiply(A=@n6) {constB=0.5})
  n8 = TextureSample!(
   UVs=Panner(
    Coordinate=MaterialFunctionCall(Vector or UVs=@n7) {materialFunction=&VectorToRadialValue}.Radial Coordinates
    Speed=AppendVector(A=@n5.B, B=@n5.B)
   ) {bFractionalPart=true}
   Tex
   "DDX(UVs)"=DDX(Value=@n7)
   "DDY(UVs)"=DDY(Value=@n7)
   Apply View MipBias
  ) {
   mipValueMode=TMVM_Derivative
   samplerType=SAMPLERTYPE_LinearColor
   texture=/Game/Effects/Fort_Effects/World/Enviroment/Lightning/lightningPainting
  }
  n9 = Multiply(
   A=@n2
   B=Multiply(
    A=@n8.R
    B=DynamicParameter() {paramNames [Square Mask Multiplier, Square Mask Exponent, Screen Contrast, Detail Multiplier]}.Detail Multiplier
   )
  )
  Multiply(
   A=@n3.A
   B=Power(
    Base=Add(A=@n2, B=@n9)
    Exp=DynamicParameter() {paramNames [Square Mask Multiplier, Square Mask Exponent, Screen Contrast, Detail Multiplier]}.Screen Contrast
   )
  )
 ]
}
