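# Unlit translucent material for Niagara ribbons, named as a simple glow for the bent baton effect.
# A color parameter, red by default, passes through the shared MF_FX_EmissiveAdjustments function as emissive.
# A mask built from V times one-minus-V fades both edges without a texture, peaking at a quarter strength in the middle, and is scaled by particle alpha.
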Material /BeastGirlStormClawsGameplay/Assets/FX/Materials/M_BentBaton_SimpleGlowRibbon
graph {
 properties {
  bAutomaticallySetUsageInEditor=false
  bUsedWithNiagaraRibbons=true
  blendMode=BLEND_Translucent
  shadingModel=MSM_Unlit
  usedShadingModels=Unlit
 }
 nodes [
  MaterialFunctionCall(Emissive=VectorParameter() {defaultValue=<rgba 1 0 0.025 1>}.RGB) {materialFunction=&MF_FX_EmissiveAdjustments}
  n1 = ComponentMask(None=TextureCoordinate()) {g=true}
  Multiply(A=Multiply(A=OneMinus(None=@n1), B=@n1), B=ParticleColor().A)
 ]
}
