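# Unlit translucent glow material for Niagara mesh particles, in the BeastGirlStormClaws effects.
# Fresnel is added to particle color for the glow, and a dynamic Thicken parameter scales the world vertex normal.
# An electric noise mask is triplanar-projected from local position, blended by contrast-adjusted pre-skinned normals.
# A time-driven sine pulses that noise, which is multiplied by particle alpha.
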Material /BeastGirlStormClawsGameplay/Assets/FX/Materials/M_BlobMesh_Glow
graph {
 properties {
  bAutomaticallySetUsageInEditor=false
  bUsedWithNiagaraMeshParticles=true
  blendMode=BLEND_Translucent
  shadingModel=MSM_Unlit
  usedShadingModels=Unlit
 }
 nodes [
  n1 = ParticleColor()
  Add(A=Fresnel(), B=@n1.RGB)
  Multiply(A=VertexNormalWS(), B=DynamicParameter() {paramNames [Thicken,Param2,Param3,Param4]}.Thicken)
  n2 = VectorParameter() {group=Derez, parameterName=DeRezSquare_TextureSize, defaultValue=<rgba 100 100 100 0>}
  n3 = Divide(A=LocalPosition() {includedOffsets=ExcludeOffsets}.XYZ, B=Multiply(A=Abs(None=@n2.RGB)) {constB=-1})
  n4 = TextureSample(UVs=ComponentMask(None=@n3) {b=true, r=true}) {
   samplerSource=SSM_Wrap_WorldGroupSettings
   samplerType=SAMPLERTYPE_LinearColor
   texture=/Game/Characters/Player/Male/Medium/Bodies/M_MED_SRDrift/Textures/T_ElectricNoise_Mask
  }
  n5 = TextureSample(UVs=ComponentMask(None=@n3) {b=true, g=true}) {
   samplerSource=SSM_Wrap_WorldGroupSettings
   samplerType=SAMPLERTYPE_LinearColor
   texture=&T_ElectricNoise_Mask
  }
  n6 = VertexInterpolator(VS=PreSkinnedNormal())
  n7 = Constant() {r=1}
  n8 = MaterialFunctionCall(In=Abs(None=ComponentMask(None=@n6.PS) {r=true}), Contrast=@n7) {materialFunction=&CheapContrast}
  n9 = TextureSample(UVs=ComponentMask(None=@n3) {g=true, r=true}) {
   samplerSource=SSM_Wrap_WorldGroupSettings
   samplerType=SAMPLERTYPE_LinearColor
   texture=&T_ElectricNoise_Mask
  }
  n10 = Abs(None=ComponentMask(None=@n6.PS) {b=true})
  n11 = LinearInterpolate(
   A=LinearInterpolate(A=@n4.RGB, B=@n5.RGB, Alpha=ComponentMask(None=@n8.Result) {r=true})
   B=@n9.RGB
   Alpha=ComponentMask(None=MaterialFunctionCall(In=@n10, Contrast=@n7) {materialFunction=&CheapContrast}.Result) {r=true}
  )
  n12 = Time()
  n13 = ScalarParameter() {defaultValue=5, parameterName=Time Scale}
  n14 = Sine(None=Add(A=@n11, B=Multiply(A=@n12, B=@n13)))
  Multiply(A=@n1.A, B=Multiply(A=ConstantBiasScale(None=@n14), B=@n11))
  Multiply()
 ]
}
