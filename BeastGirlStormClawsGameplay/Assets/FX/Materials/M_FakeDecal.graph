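# Translucent unlit material for Niagara sprites that, by its name, fakes a decal with a glowing, particle-colored effect.
# Emissive sums a base tint, a spinning BentBaton star from clamped, scaled UVs and an inverted DepthFade glow near intersecting geometry, feeding MF_FX_EmissiveAdjustments.
# A radial SphereMask gets a wobbly edge from UVs offset by a scrolling cloth-sim texture sampled in world space.
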Material /BeastGirlStormClawsGameplay/Assets/FX/Materials/M_FakeDecal
graph {
 properties {
  bAutomaticallySetUsageInEditor=false
  bUsedWithNiagaraSprites=true
  blendMode=BLEND_Translucent
  shadingModel=MSM_Unlit
  usedShadingModels=Unlit
 }
 nodes [
  n1 = ParticleColor()
  n2 = Multiply(A=@n1.RGB, B=ScalarParameter() {defaultValue=0.8})
  n3 = TextureCoordinate()
  n4 = ScalarParameter() {defaultValue=3, parameterName=Param_6}
  n5 = Rotator(
   Coordinate=Subtract(A=Multiply(A=@n3, B=@n4), B=ScalarParameter() {defaultValue=1, parameterName=Param_7})
   Time=Multiply(A=Time(), B=ScalarParameter() {defaultValue=3.75, parameterName=Param_8})
  )
  n6 = Multiply(
   A=@n1.RGB
   B=Multiply(
    A=TextureSample(UVs=@n5) {samplerSource=SSM_Clamp_WorldGroupSettings, texture=&/T_BentBaton_Star}.A
    B=ScalarParameter() {defaultValue=0.75, parameterName=Star Strength}
   )
  )
  n7 = ScalarParameter() {defaultValue=35, parameterName=Param_10}
  n8 = Multiply(A=@n1.RGB, B=OneMinus(None=DepthFade(FadeDistance=@n7) {fadeDistanceDefault=20}))
  n9 = ScalarParameter() {defaultValue=5, parameterName=Param_9}
  MaterialFunctionCall(Emissive=Add(A=Add(A=@n2, B=@n6), B=Multiply(A=@n8, B=@n9))) {materialFunction=&MF_FX_EmissiveAdjustments}
  n10 = Time()
  n11 = VectorParameter() {parameterName=Param_1, defaultValue=<rgba 0 0 -50 0>}
  n12 = Multiply(A=Add(A=WorldPosition().XYZ, B=Multiply(A=@n10, B=@n11.RGB)), B=ScalarParameter() {defaultValue=0.0025, parameterName=Param_5})
  n13 = Power(
   Base=OneMinus(None=TextureSample(UVs=@n12) {samplerType=SAMPLERTYPE_LinearColor, texture=/Game/Global/MaterialFunctions/VT_ClothSimWPO}.G)
   Exp=ScalarParameter() {defaultValue=2, parameterName=Param_4}
  )
  n14 = ScalarParameter() {defaultValue=0.05, parameterName=Param_3}
  SphereMask(
   A=Add(A=Multiply(A=@n13, B=@n14), B=TextureCoordinate())
   B=ComponentMask(None=VectorParameter() {defaultValue=<rgba 0.5 0.5 0 0>}.RGB) {g=true, r=true}
   Radius=ScalarParameter() {defaultValue=0.5, parameterName=Param_1}
   Hardness=ScalarParameter() {parameterName=Param_2}
  )
 ]
}
