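# Unlit translucent material for Niagara ribbons, named as a bent-baton ground indicator, that draws animated plasma lightning.
# Two panning plasma samples, distorted by scrolling electric noise, are summed to shape opacity and a normal-direction offset.
# Opacity also uses particle alpha and a V-coordinate mask that fades both ribbon edges.
# Particle color drives emissive, with an inverted depth fade boosting glow where the ribbon nears scene geometry.
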Material /BeastGirlStormClawsGameplay/Assets/FX/Materials/M_Indicator_Ground_BentBaton
graph {
 properties {
  bAutomaticallySetUsageInEditor=false
  bUsedWithNiagaraRibbons=true
  blendMode=BLEND_Translucent
  shadingModel=MSM_Unlit
  usedShadingModels=Unlit
 }
 nodes [
  n1 = ParticleColor()
  n2 = Multiply(A=@n1.RGB, B=ScalarParameter() {defaultValue=35, parameterName=Param_5})
  n3 = OneMinus(None=DepthFade(FadeDistance=ScalarParameter() {defaultValue=60, parameterName=Param_10}) {fadeDistanceDefault=20})
  MaterialFunctionCall(Emissive=Add(A=Multiply(A=@n2, B=@n3), B=@n1.RGB)) {materialFunction=&MF_FX_EmissiveAdjustments}
  n4 = TextureCoordinate()
  n5 = ComponentMask(None=VectorParameter() {defaultValue=<rgba 5 1 0 0>}.RGB) {g=true, r=true}
  n6 = MaterialFunctionCall(XY=Panner(Coordinate=Multiply(A=@n4, B=@n5)) {speedX=1}) {materialFunction=&Swizzle}
  n7 = TextureCoordinate()
  n8 = ScalarParameter() {defaultValue=0.5, parameterName=Param_4}
  n9 = Time()
  n10 = ScalarParameter() {defaultValue=0.5, parameterName=Param_3}
  n11 = TextureSample(UVs=Panner(Coordinate=Multiply(A=@n7, B=@n8), Time=Multiply(A=@n9, B=@n10)) {speedX=-1}) {samplerType=SAMPLERTYPE_LinearColor, texture=/Game/Characters/Player/Male/Medium/Bodies/M_MED_SRDrift/Textures/T_ElectricNoise}
  n12 = Multiply(A=@n11.G, B=ScalarParameter() {defaultValue=0.01, parameterName=Param_2})
  n13 = Multiply(A=@n11.G, B=ScalarParameter() {defaultValue=0.05})
  n14 = TextureCoordinate()
  n15 = ComponentMask(None=VectorParameter() {defaultValue=<rgba 5 1 0 0>}.RGB) {g=true, r=true}
  n16 = MaterialFunctionCall(XY=Panner(Coordinate=Multiply(A=@n14, B=@n15)) {speedX=0.5}) {materialFunction=&Swizzle}
  n17 = Add(
   A=TextureSample(UVs=Add(A=@n6.YX, B=@n12)) {texture=/Game/Effects/Fort_Effects/Textures/Tile/T_Lightning_Plasma_Soft}.G
   B=TextureSample(UVs=Add(A=@n13, B=@n16.YX)) {texture=&T_Lightning_Plasma_Soft}.G
  )
  Multiply(A=Multiply(A=@n17, B=VertexNormalWS()), B=ScalarParameter() {defaultValue=10, parameterName=Param_1})
  n18 = ComponentMask(None=TextureCoordinate()) {g=true}
  n19 = OneMinus(None=@n18)
  Multiply(A=@n1.A, B=Multiply(A=Multiply(A=@n19, B=@n18), B=@n17))
 ]
}
