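# Unlit translucent lightning arc material for mesh particles, stretching arcs toward a target through world position offset.
# SplineThicken sets width, masked time-driven sines add mid-arc wobble, and a UV-weighted pull targets the randomly offset actor or Niagara position (PosX also feeds Z).
# Panning plasma texture drives HDR color and depth-faded opacity, with optional masks and intersection brightening.
# Further duplicated networks hold spline-texture glow, growth and dissolve logic.
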Material /BeastGirlStormClawsGameplay/Assets/FX/Materials/M_Lightning_Arc_WPO_Stretch
graph {
 properties {
  bUsedWithMeshParticles=true
  bUsedWithNiagaraMeshParticles=true
  blendMode=BLEND_Translucent
  pixelDepthOffsetMode=PDOM_Legacy
  shadingModel=MSM_Unlit
  twoSided=true
  usedShadingModels=Unlit
 }
 nodes [
  n1 = DynamicParameter() {paramNames [Width,WobbleMult,Random1,Random2]}
  n2 = Multiply(A=ScalarParameter() {defaultValue=6, parameterName=Width}, B=@n1.Width)
  n3 = ScalarParameter() {parameterName=WidthTipMult}
  n4 = TextureCoordinate()
  n5 = Constant() {r=0.5}
  n6 = SphereMask(A=ComponentMask(None=@n4) {g=true}, B=@n5, Radius=@n5) {hardnessPercent=50}
  n7 = TextureCoordinate()
  n8 = Add(A=@n1.Random2, B=Time())
  n9 = Multiply(A=@n8, B=ScalarParameter() {defaultValue=3, parameterName=WobbleSpeed})
  n10 = ComponentMask(None=Add(A=@n7, B=Multiply(A=@n9) {constB=-0.6})) {g=true}
  n11 = Multiply(A=Sine(None=@n10) {period=0.2}) {constB=0.2}
  n12 = ComponentMask(None=Add(A=@n7, B=Multiply(A=@n9) {constB=-1})) {g=true}
  n13 = Sine(None=@n12) {period=0.8}
  n14 = Sine(None=@n10) {period=0.23}
  n15 = AppendVector(A=AppendVector(A=Constant(), B=Add(A=@n11, B=@n13)), B=Add(A=Multiply(A=@n14) {constB=0.2}, B=Sine(None=@n12) {period=0.6}))
  n16 = Multiply(
   A=Normalize(VectorInput=Transform(None=@n15) {transformSourceType=TRANSFORMSOURCE_Local})
   B=Multiply(A=ScalarParameter() {defaultValue=5, parameterName=Wobble}, B=@n1.WobbleMult)
  )
  n17 = ScalarParameter() {defaultValue=1, parameterName=StetchWPOMult}
  n18 = DynamicParameter() {desc=Index 2 parameter for Niagara, parameterIndex=1, defaultValue=<rgba 0 0 0 0>, paramNames [PosX,PosY,PosZ,Param4]}
  n19 = Multiply(
   A=Subtract(A=AppendVector(A=@n1.Random1, B=@n1.Random1)) {constB=0.5}
   B=ScalarParameter() {defaultValue=30, parameterName=ActorPositionOffset}
  ) {constB=10}
  n20 = Constant()
  n21 = Add(
   A=LinearInterpolate(
    A=ActorPositionWS()
    B=MaterialFunctionCall(X=@n18.PosX, Y=@n18.PosX, Z=@n18.PosX) {materialFunction=&MakeFloat3}.Result
    Alpha=ScalarParameter() {parameterName=UseNiagaraPositionParam}
   )
   B=Add(A=AppendVector(A=@n19, B=@n20), B=VectorParameter() {parameterName=ActorPositionAdd}.RGB)
  )
  n22 = WorldPosition()
  n23 = ComponentMask(None=TextureCoordinate()) {g=true}
  n24 = Multiply(A=Multiply(A=Subtract(A=@n21, B=@n22.XYZ), B=OneMinus(None=@n23)), B=Clamp(None=LinearInterpolate() {constA=2, constB=0}, Min=@n18.Param4))
  Add(
   A=MaterialFunctionCall(WidthBase=@n2, WidthTip=Multiply(A=@n2, B=@n3) {constB=5}, UVs for Thickness=Power(Base=@n4) {constExponent=0.6}) {materialFunction=&SplineThicken}.WorldPositionOffset
   B=Add(A=Multiply(A=@n6, B=@n16), B=Multiply(A=@n17, B=@n24))
  )
  DynamicParameter() {parameterIndex=3, paramNames [Width,WobbleMult,Random1,Random2]}
  n25 = ComponentMask(None=TextureCoordinate()) {g=true}
  n26 = StaticSwitchParameter(True=OneMinus(None=@n25), False=@n25) {parameterName=AlphaMasks_Invert}
  n27 = ScalarParameter() {defaultValue=2, parameterName=AlphaMask_Power}
  n28 = Saturate(None=Multiply(A=Power(Base=@n26, Exp=@n27), B=ScalarParameter() {defaultValue=3, parameterName=AlphaMask_Multiply}) {constB=3})
  n29 = TextureCoordinate() {vTiling=0.5}
  n30 = AppendVector(A=Constant() {r=1}, B=LinearInterpolate(Alpha=@n1.Random1) {constA=0.5, constB=1.2})
  n31 = ScalarParameter() {defaultValue=1, parameterName=PanSpeed}
  n32 = TextureSample(UVs=Panner(Coordinate=Multiply(A=@n29, B=@n30), Time=Multiply(A=@n8, B=@n31)) {speedY=-1}) {texture=/Game/Effects/Fort_Effects/Textures/Tile/T_Lightning_Plasma_Soft}
  n33 = ParticleColor()
  n34 = Multiply(A=Add(A=@n32.R, B=Multiply(A=@n32.B) {constB=0.2}), B=@n33.A)
  n35 = Multiply(A=StaticSwitchParameter(True=Multiply(A=@n28, B=@n34), False=@n34) {parameterName=AlphaMasks}, B=NamedRerouteUsage())
  n36 = ScalarParameter() {group=Opacity, parameterName=Depth Fade}
  FeatureLevelSwitch(Default=QualitySwitch(Default=DepthFade(Opacity=@n35, FadeDistance=@n36), Low=@n35), ES2=@n35, ES3_1=@n35)
  n37 = Multiply(
   A=@n33.RGB
   B=LinearInterpolate(
    A=VectorParameter() {parameterName=EdgeColor, defaultValue=<rgba 0.2226 0.2425 0.5 0>}.RGB
    B=VectorParameter() {parameterName=Color, defaultValue=<rgba 0.6214 0.9501 1 0>}.RGB
    Alpha=@n32.G
   )
  )
  n38 = MaterialFunctionCall(LDR Input=Clamp(None=@n37), HDR Tint=@n37, HDR Intensity Multiplier=ScalarParameter() {defaultValue=20, parameterName=HDRIntensity}) {materialFunction=&DeriveHDRfromLDR}
  n39 = DepthFade(FadeDistance=ScalarParameter() {defaultValue=100, parameterName=IntesectingThickness})
  n40 = LinearInterpolate(
   A=Multiply(A=@n38.RemappedRange, B=ScalarParameter() {defaultValue=2, parameterName=IntesectingEmissiveBoost})
   B=@n38.RemappedRange
   Alpha=@n39
  )
  MaterialFunctionCall(Emissive=StaticSwitchParameter(
    True=MaterialFunctionCall!(Default=@n40, Low/Mobile=@n38.RemappedRange) {materialFunction=/Game/Effects/Fort_Effects/Materials/_Functions/MF_LowMobile_PathingSwitch}.Result
    False=@n38.RemappedRange
   ) {parameterName=MeshIntersectingColorMultiply}) {materialFunction=&MF_FX_EmissiveAdjustments}
  n41 = TextureCoordinate()
  n42 = ComponentMask(None=VectorParameter() {defaultValue=<rgba 1 4 0 0>}.RGB) {g=true, r=true}
  n43 = TextureSample(UVs=Panner(Coordinate=Multiply(A=@n41, B=@n42)) {speedY=1}) {texture=/Game/Effects/Fort_Effects/GameplayPulginsAssets/ToonLightning/MasterMaterial/V1/T_LightningSplineTexture}
  n44 = ScalarParameter() {defaultValue=4}
  n45 = Add(
   A=LinearInterpolate(B=VectorParameter() {parameterName=Primary Color, defaultValue=<rgba 5 0 0.0833 1>}.RGB, Alpha=@n43.R)
   B=LinearInterpolate(B=VectorParameter() {parameterName=Param_2, defaultValue=<rgba 5 5 5 0>}.RGB, Alpha=Power(Base=@n43.G, Exp=@n44))
  )
  MaterialFunctionCall(LDR Input=Clamp(None=@n45), HDR Tint=@n45, HDR Intensity Multiplier=ScalarParameter() {defaultValue=4, parameterName=HDR Intensity}) {materialFunction=&DeriveHDRfromLDR}
  n46 = MaterialFunctionCall(Float2=NamedRerouteUsage()) {materialFunction=&BreakOutFloat2Components}
  n47 = VectorParameter() {parameterName=BottomBright_Refine, defaultValue=<rgba 4 10 0 0.5>}
  n48 = MaterialFunctionCall(Texture=@n46.G, RGB=@n47.RGB, A=@n47.RGB) {materialFunction=&MF_FX_Refine}
  n49 = VectorParameter() {parameterName=TipBright_Refine, defaultValue=<rgba 4 10 0 0.5>}
  n50 = Add(
   A=@n48.Result
   B=MaterialFunctionCall(Texture=OneMinus(None=@n46.G), RGB=@n49.RGB, A=@n49.RGB) {materialFunction=&MF_FX_Refine}.Result
  )
  n51 = ScalarParameter() {parameterName=UseTipBright}
  n52 = ScalarParameter() {defaultValue=0.1, parameterName=SpikeBrightness_SpikeDataSubtract}
  n53 = Saturate(None=Subtract(A=NamedRerouteUsage(), B=@n52))
  n54 = OneMinus(None=@n52)
  n55 = Multiply(
   A=Power(Base=Divide(A=@n53, B=@n54), Exp=ScalarParameter() {defaultValue=10, parameterName=SpikeBrightness_SpikeDataPower})
   B=ScalarParameter() {parameterName=SpikeBrightness_Scale}
  )
  n56 = NamedRerouteUsage()
  n57 = VectorParameter() {group=em, parameterName=Glow_EdgeMask_Control, channelNames {a="", b=Power, g=Multiply, r=Offset}, defaultValue=<rgba 0 2 2 1>}
  n58 = VectorParameter() {parameterName=Glow_Refine, defaultValue=<rgba 4 2 0.5 0.5>}
  n59 = MaterialFunctionCall(
   Texture=MaterialFunctionCall!(
    UorV=MaterialFunctionCall(Float2=@n56) {materialFunction=&BreakOutFloat2Components}.R
    EdgeControl=@n57
   ) {materialFunction=/Game/Effects/Niagara/MasterMaterials/Functions/Common/FXCommon_EdgeMask_Simple}.Mask
   RGB=@n58.RGB
   A=@n58.RGB
  ) {materialFunction=&MF_FX_Refine}
  n60 = VectorParameter() {parameterName=Glow_ColorA, defaultValue=<rgba 0.004 0.053 0.2 1>}
  n61 = VectorParameter() {parameterName=Glow_ColorB, defaultValue=<rgba 0.04 0.13 0.4 1>}
  n62 = NamedRerouteUsage()
  n63 = Power(Base=@n62, Exp=ScalarParameter() {defaultValue=3, parameterName=Glow_ColorFromSpike_Power})
  n64 = Power(Base=@n62, Exp=ScalarParameter() {defaultValue=2, parameterName=Glow_BrogjtmessFromSpike_Power})
  n65 = ScalarParameter() {defaultValue=0.1, parameterName=Glow_BrogjtmessFromSpike_Str}
  n66 = Multiply(A=Multiply(A=Add(A=@n59.Result), B=LinearInterpolate(A=@n60.RGB, B=@n61.RGB, Alpha=@n63)), B=Add(A=Multiply(A=@n64, B=@n65)))
  n67 = TextureObjectParameter() {
   group=LightningT
   parameterName=LightningT_Texture
   sortPriority=0
   texture=&T_LightningSplineTexture
  }
  n68 = DynamicParameter() {
   parameterIndex=2
   defaultValue=<rgba 5 0 500 30>
   paramNames [ScaleSplineWidth, BaseWitdthOffset, TextureRandom, MoveAlignLightningDirSphereMaskSize]
  }
  n69 = VectorParameter() {group=LightningT, parameterName=LightningT_UVControl, channelNames {a=VPan, b=UPan, g=VScale, r=UScale}, defaultValue=<rgba 0.5 1.5 0 0>}
  n70 = MaterialFunctionCall!(
   Texture=@n67
   AdditionalUVOffset=MaterialFunctionCall(X=Constant(), Y=@n68.TextureRandom) {materialFunction=&MakeFloat2}.Result
   UVControl_RGB=@n69
   UVControl_A=@n69
   "[Override]TimeInput"
   "[Override]ScaleUVByCenter"=StaticBool() {value=true}
   "[Override]UVInput"=NamedRerouteUsage()
  ) {materialFunction=/Game/Effects/Niagara/MasterMaterials/Functions/Common/FXCommon_TextureSample_V2}
  n71 = VectorParameter() {parameterName=InnerGlowRefine, defaultValue=<rgba 5 2 1 0.5>}
  n72 = Multiply(
   A=MaterialFunctionCall(Texture=OneMinus(None=@n70.G), RGB=@n71.RGB, A=@n71.RGB) {materialFunction=&MF_FX_Refine}.Result
   B=@n70.B
  )
  n73 = LinearInterpolate(
   A=VectorParameter() {parameterName=GlowColor, defaultValue=<rgba 0 0.8333 5 1>}.RGB
   B=LinearInterpolate(
    A=VectorParameter() {parameterName=BoltColor, defaultValue=<rgba 0 0.8333 1.25 1>}.RGB
    B=VectorParameter() {parameterName=InnerGlowColor, defaultValue=<rgba 0.65 4.7667 13 5>}.RGB
    Alpha=Saturate(None=@n72)
   )
   Alpha=@n70.B
  )
  n74 = Multiply(
   A=Add(A=LinearInterpolate(A=@n48.Result, B=@n50, Alpha=@n51), B=ScalarParameter() {defaultValue=1, parameterName=BottomBright_Base})
   B=Multiply(A=Add(A=@n55), B=StaticSwitchParameter(True=@n66, False=@n73) {parameterName=Glow})
  )
  n75 = Constant3Vector() {constant=<rgba 1 0 0 0>}
  n76 = DynamicParameter() {
   parameterIndex=3
   defaultValue=<rgba 0 0 1 -0.7>
   paramNames [MoveAlignLightningDir, "MoveToCenter(NoPerf/Mobile)", Growth, ThicknessDissolve]
  }
  n77 = Subtract(A=NamedRerouteUsage(), B=@n76.ThicknessDissolve)
  n78 = ScalarParameter() {defaultValue=2, parameterName=ThicknessDissolveHardness}
  n79 = Floor(None=Saturate(None=Multiply(A=@n77, B=@n78)))
  n80 = Multiply(
   A=LinearInterpolate(A=@n74, B=LinearInterpolate(A=@n75, B=@n74, Alpha=@n79), Alpha=ScalarParameter() {parameterName=DebugThicknessDissolve})
   B=ParticleColor().RGB
  )
  n81 = AppendVector(A=TextureCoordinate(), B=Constant())
  n82 = ScalarParameter() {parameterName=DebugUV}
  MaterialFunctionCall(Emissive=Divide(A=LinearInterpolate(A=@n80, B=@n81, Alpha=@n82)) {constB=2.1}) {materialFunction=&MF_FX_EmissiveAdjustments}
  n83 = ParticleColor()
  n84 = TextureSample() {samplerType=SAMPLERTYPE_Masks, texture=/Game/Characters/Player/Male/Medium/Bodies/M_MED_Storm_Soldier/Textures/ElectricityNoise_M}
  Multiply(A=Multiply(A=@n43.R, B=LinearInterpolate(A=@n83.A, Alpha=@n84.G)), B=ScalarParameter() {defaultValue=1, parameterName=Opacity})
  n85 = TextureCoordinate()
  n86 = Power(
   Base=Sine(None=ComponentMask(None=@n85) {g=true}) {period=2}
   Exp=ScalarParameter() {defaultValue=1, group=Opacity, parameterName=Ramp Mask Power}
  )
  NamedRerouteDeclaration(None=If(A=ScalarParameter() {group=Opacity, parameterName=Enable Ramp Mask}, "A > B"=@n86, "A == B"=@n86, "A < B"=Constant() {r=1}) {constB=0.99}) {name=Ramp Fade, nodeColor=<rgba 0 1 0.3203 1>}
  n87 = MaterialFunctionCall(Float2=NamedRerouteUsage()) {materialFunction=&BreakOutFloat2Components}
  n88 = VectorParameter() {group=SmpGlow, parameterName=SmpGlow_EdgeMask_Control, channelNames {a="", b=Power, g=Multiply, r=Offset}, defaultValue=<rgba -1 1 1 1>}
  n89 = MaterialFunctionCall(UorV=@n87.R, EdgeControl=@n88) {materialFunction=&FXCommon_EdgeMask_Simple}
  n90 = NamedRerouteUsage()
  n91 = VectorParameter() {group=em, parameterName=Glow_EdgeMask_Control, channelNames {a="", b=Power, g=Multiply, r=Offset}, defaultValue=<rgba 0 2 2 1>}
  n92 = VectorParameter() {parameterName=Glow_Refine, defaultValue=<rgba 4 2 0.5 0.5>}
  n93 = MaterialFunctionCall(
   Texture=MaterialFunctionCall(
    UorV=MaterialFunctionCall(Float2=@n90) {materialFunction=&BreakOutFloat2Components}.R
    EdgeControl=@n91
   ) {materialFunction=&FXCommon_EdgeMask_Simple}.Mask
   RGB=@n92.RGB
   A=@n92.RGB
  ) {materialFunction=&MF_FX_Refine}
  n94 = MaterialFunctionCall(Float2=NamedRerouteUsage()) {materialFunction=&BreakOutFloat2Components}
  n95 = VectorParameter() {
   group=UPDOWNFADE
   parameterName=GlowFade_EdgeMask_InvertControl
   channelNames {a="", b=Power, g=Multiply, r=Offset}
   defaultValue=<rgba 0 20 2 1>
  }
  n96 = VectorParameter() {group=UPDOWNFADE, parameterName=GlowFade_EdgeMask_Control, channelNames {a="", b=Power, g=Multiply, r=Offset}, defaultValue=<rgba 0 2 1 1>}
  n97 = TextureObjectParameter() {
   group=LightningT
   parameterName=LightningT_Texture
   sortPriority=0
   texture=&T_LightningSplineTexture
  }
  n98 = DynamicParameter() {
   parameterIndex=2
   defaultValue=<rgba 5 0 500 30>
   paramNames [ScaleSplineWidth, BaseWitdthOffset, TextureRandom, MoveAlignLightningDirSphereMaskSize]
  }
  n99 = VectorParameter() {group=LightningT, parameterName=LightningT_UVControl, channelNames {a=VPan, b=UPan, g=VScale, r=UScale}, defaultValue=<rgba 0.5 1.5 0 0>}
  n100 = MaterialFunctionCall(
   Texture=@n97
   AdditionalUVOffset=MaterialFunctionCall(X=Constant(), Y=@n98.TextureRandom) {materialFunction=&MakeFloat2}.Result
   UVControl_RGB=@n99
   UVControl_A=@n99
   "[Override]ScaleUVByCenter"=StaticBool() {value=true}
   "[Override]UVInput"=NamedRerouteUsage()
  ) {materialFunction=&FXCommon_TextureSample_V2}
  n101 = VectorParameter() {parameterName=OutGlowRefine, defaultValue=<rgba 1.5 0.5 0 0>}
  n102 = StaticSwitchParameter(
   True=Multiply(
    A=@n93.Result
    B=MaterialFunctionCall!(UorV=@n94.G, Invert_EdgeControl=@n95, NoInvert_EdgeControl=@n96) {materialFunction=/Game/Effects/Niagara/MasterMaterials/Functions/Common/FXCommon_EdgeMask_Ex}.Mask
   )
   False=MaterialFunctionCall(Texture=@n100.R, RGB=@n101.RGB, A=@n101.RGB) {materialFunction=&MF_FX_Refine}.Result
  ) {parameterName=Glow}
  n103 = MaterialFunctionCall(Float2=NamedRerouteUsage()) {materialFunction=&BreakOutFloat2Components}
  n104 = LinearInterpolate(A=@n103.G, B=OneMinus(None=@n103.G), Alpha=ScalarParameter() {defaultValue=1, parameterName=SplineThickness_GrowthDirectionInvert})
  n105 = DynamicParameter() {
   parameterIndex=3
   defaultValue=<rgba 0 0 1 -0.7>
   paramNames [MoveAlignLightningDir, "MoveToCenter(NoPerf/Mobile)", Growth, ThicknessDissolve]
  }
  n106 = Saturate(None=@n105.Growth)
  n107 = DynamicParameter() {paramNames [MaxGrowthLength, NewLightningTech_InvertDir, WPORand, NewLightningTech_Index]}
  n108 = ScalarParameter() {defaultValue=5, parameterName=SplineThickness_GrowthGradient}
  n109 = Divide(B=@n108)
  n110 = Add(A=Subtract(A=@n104), B=Multiply(A=Multiply(A=@n106, B=@n107.MaxGrowthLength), B=Add(B=@n109) {constA=1}))
  n111 = Add(A=@n104, B=@n105.Growth)
  n112 = StaticSwitchParameter(True=Saturate(None=Multiply(A=@n110, B=@n108)), False=Saturate(None=Multiply(A=@n111, B=@n108))) {parameterName=UseNewGrowthMethod}
  n113 = VectorParameter() {
   group=EdgeFade
   parameterName=EdgeFade_EdgeMask_InvertControl
   channelNames {a="", b=Power, g=Multiply, r=Offset}
   defaultValue=<rgba 5 1 1 1>
  }
  n114 = VectorParameter() {group=EdgeFade, parameterName=EdgeFade_EdgeMask_Control, channelNames {a="", b=Power, g=Multiply, r=Offset}, defaultValue=<rgba 5 1 1 1>}
  n115 = Multiply(
   A=Add(A=Multiply(A=@n89.Mask, B=@n88.A), B=Multiply(A=@n112, B=@n102)) {constB=0}
   B=LinearInterpolate(
    B=MaterialFunctionCall(UorV=@n87.G, Invert_EdgeControl=@n113, NoInvert_EdgeControl=@n114) {materialFunction=&FXCommon_EdgeMask_Ex}.Mask
    Alpha=ScalarParameter() {defaultValue=1, group=EdgeFade, parameterName=Use edge fade mask Control}
   ) {constA=1}
  )
  n116 = ScalarParameter() {parameterName=DF}
  n117 = ParticleColor()
  n118 = Multiply(A=Multiply(A=NamedRerouteUsage(), B=DepthFade(Opacity=@n115, FadeDistance=@n116)), B=@n117.A)
  n119 = ScalarParameter() {defaultValue=1, parameterName=DepthFade}
  n120 = MaterialFunctionCall(Distance=ScalarParameter() {parameterName=NearCameraFade}) {materialFunction=&Near_Camera_Fade}
  n121 = ScalarParameter() {defaultValue=2, parameterName=NearCameraFadePower}
  NamedRerouteDeclaration(None=Multiply(A=DepthFade(Opacity=@n118, FadeDistance=@n119), B=Power(Base=@n120.Result, Exp=@n121))) {name=Opacity, nodeColor=<rgba 0.7344 1 0 1>}
  n122 = MaterialFunctionCall(Float2=NamedRerouteUsage()) {materialFunction=&BreakOutFloat2Components}
  n123 = VectorParameter() {parameterName=BottomBright_Refine, defaultValue=<rgba 4 10 0 0.5>}
  n124 = MaterialFunctionCall(Texture=@n122.G, RGB=@n123.RGB, A=@n123.RGB) {materialFunction=&MF_FX_Refine}
  n125 = VectorParameter() {parameterName=TipBright_Refine, defaultValue=<rgba 4 10 0 0.5>}
  n126 = Add(
   A=@n124.Result
   B=MaterialFunctionCall(Texture=OneMinus(None=@n122.G), RGB=@n125.RGB, A=@n125.RGB) {materialFunction=&MF_FX_Refine}.Result
  )
  n127 = ScalarParameter() {parameterName=UseTipBright}
  n128 = ScalarParameter() {defaultValue=0.1, parameterName=SpikeBrightness_SpikeDataSubtract}
  n129 = Saturate(None=Subtract(A=NamedRerouteUsage(), B=@n128))
  n130 = OneMinus(None=@n128)
  n131 = Multiply(
   A=Power(Base=Divide(A=@n129, B=@n130), Exp=ScalarParameter() {defaultValue=10, parameterName=SpikeBrightness_SpikeDataPower})
   B=ScalarParameter() {parameterName=SpikeBrightness_Scale}
  )
  n132 = VectorParameter() {parameterName=Glow_ColorA, defaultValue=<rgba 0.004 0.053 0.2 1>}
  n133 = VectorParameter() {parameterName=Glow_ColorB, defaultValue=<rgba 0.04 0.13 0.4 1>}
  n134 = NamedRerouteUsage()
  n135 = Power(Base=@n134, Exp=ScalarParameter() {defaultValue=3, parameterName=Glow_ColorFromSpike_Power})
  n136 = Power(Base=@n134, Exp=ScalarParameter() {defaultValue=2, parameterName=Glow_BrogjtmessFromSpike_Power})
  n137 = ScalarParameter() {defaultValue=0.1, parameterName=Glow_BrogjtmessFromSpike_Str}
  n138 = Multiply(A=Multiply(A=Add(A=@n93.Result), B=LinearInterpolate(A=@n132.RGB, B=@n133.RGB, Alpha=@n135)), B=Add(A=Multiply(A=@n136, B=@n137)))
  n139 = VectorParameter() {parameterName=InnerGlowRefine, defaultValue=<rgba 5 2 1 0.5>}
  n140 = Multiply(
   A=MaterialFunctionCall(Texture=OneMinus(None=@n100.G), RGB=@n139.RGB, A=@n139.RGB) {materialFunction=&MF_FX_Refine}.Result
   B=@n100.B
  )
  n141 = LinearInterpolate(
   A=VectorParameter() {parameterName=GlowColor, defaultValue=<rgba 0 0.8333 5 1>}.RGB
   B=LinearInterpolate(
    A=VectorParameter() {parameterName=BoltColor, defaultValue=<rgba 0 0.8333 1.25 1>}.RGB
    B=VectorParameter() {parameterName=InnerGlowColor, defaultValue=<rgba 0.65 4.7667 13 5>}.RGB
    Alpha=Saturate(None=@n140)
   )
   Alpha=@n100.B
  )
  n142 = Multiply(
   A=Add(A=LinearInterpolate(A=@n124.Result, B=@n126, Alpha=@n127), B=ScalarParameter() {defaultValue=1, parameterName=BottomBright_Base})
   B=Multiply(A=Add(A=@n131), B=StaticSwitchParameter(True=@n138, False=@n141) {parameterName=Glow})
  )
  n143 = DynamicParameter() {
   parameterIndex=3
   defaultValue=<rgba 0 0 1 -0.7>
   paramNames [MoveAlignLightningDir, "MoveToCenter(NoPerf/Mobile)", Growth, ThicknessDissolve]
  }
  n144 = Multiply(A=Subtract(A=NamedRerouteUsage(), B=@n143.ThicknessDissolve), B=ScalarParameter() {defaultValue=2, parameterName=ThicknessDissolveHardness})
  n145 = LinearInterpolate(A=Constant3Vector() {constant=<rgba 1 0 0 0>}, B=@n142, Alpha=Floor(None=Saturate(None=@n144)))
  n146 = ScalarParameter() {parameterName=DebugThicknessDissolve}
  n147 = LinearInterpolate(
   A=Multiply(A=LinearInterpolate(A=@n142, B=@n145, Alpha=@n146), B=@n117.RGB)
   B=AppendVector(A=TextureCoordinate(), B=Constant())
   Alpha=ScalarParameter() {parameterName=DebugUV}
  )
  NamedRerouteDeclaration(None=MaterialFunctionCall(Emissive=Divide(A=@n147) {constB=2.1}) {materialFunction=&MF_FX_EmissiveAdjustments}.Result) {name=Emissive, nodeColor=<rgba 0.9453 0 1 1>}
  n148 = TextureObjectParameter() {
   group=DissolveTexture
   parameterName=DissolveTexture_Texture
   samplerType=SAMPLERTYPE_LinearColor
   sortPriority=0
   texture=/Engine/Functions/Engine_MaterialFunctions02/ExampleContent/Textures/OffsetNoiseDistanceFields_ForNormals
  }
  n149 = VectorParameter() {
   group=DissolveTexture
   parameterName=DissolveTexture_UVControl
   channelNames {a=VPan, b=UPan, g=VScale, r=UScale}
   defaultValue=<rgba 0.5 1.5 0 0.1>
  }
  n150 = MaterialFunctionCall(Texture=@n148, UVControl_RGB=@n149, UVControl_A=@n149, "[Override]UVInput"=NamedRerouteUsage()) {materialFunction=&FXCommon_TextureSample}
  n151 = DynamicParameter() {
   parameterIndex=1
   defaultValue=<rgba 0 2 -1 0>
   paramNames [NewLightningTech_WPOStr, MoveAlignLightningDirSphereMaskPower, TextureDissolve, TextureWPONoise]
  }
  n152 = Subtract(A=@n151.TextureDissolve, B=Multiply(A=NamedRerouteUsage(), B=ScalarParameter() {defaultValue=0.3, parameterName=Dissolve_ScaleWidthFactor}))
  NamedRerouteDeclaration(None=MaterialFunctionCall(In=Subtract(A=@n150.R, B=@n152), Contrast=ScalarParameter() {defaultValue=10, parameterName=Dissolve_Contrast}) {materialFunction=&CheapContrast}.Result) {name=Dissolve, nodeColor=<rgba 0.2422 0 1 1>}
  ParticleColor()
  ParticleRelativeTime()
  Saturate(None=Power(Base=@n39))
 ]
}
