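# Translucent, unlit, two-sided Niagara mesh particle material for a glowing, red-by-default teeth effect, built from BentBaton pickaxe grayscale textures.
# Opposite-panning samples of each teeth texture are blended and layered into alpha, glow, light and center, with shared UV, mask and stepped-time reroutes.
# Emissive adds noise and particle color; opacity applies end, view-angle and depth fades.
# Panning smoke noise drives world position offset, with optional SplineThicken.
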Material /BeastGirlStormClawsGameplay/Assets/FX/Materials/M_Pickaxe_BentBaton_Teeth
graph {
 properties {
  bAutomaticallySetUsageInEditor=false
  bOutputTranslucentVelocity=true
  bUsedWithNiagaraMeshParticles=true
  blendMode=BLEND_Translucent
  shadingModel=MSM_Unlit
  twoSided=true
  usedShadingModels=Unlit
 }
 nodes [
  n1 = ScalarParameter() {defaultValue=10, parameterName="WidthBase (S)"}
  n2 = StaticBool()
  n3 = TextureCoordinate() {uTiling=3, vTiling=0.5}
  n4 = ScalarParameter() {defaultValue=-1, parameterName=WPOSpeed}
  n5 = Constant2Vector() {r=1}
  n6 = Multiply(A=Multiply(A=@n4) {constB=-0.5}, B=@n5)
  n7 = Multiply(A=Multiply(A=@n4) {constB=0.5}, B=@n5)
  n8 = LinearInterpolate(
   A=TextureSample(UVs=Panner(Coordinate=@n3, Speed=@n6) {speedX=-0.5}) {texture=&T_Inky_Smoke_Tile}.R
   B=TextureSample(UVs=Panner(Coordinate=@n3, Speed=@n7) {speedX=0.5}) {texture=&T_Inky_Smoke_Tile}.R
   Alpha=NamedRerouteUsage()
  )
  n9 = VectorParameter() {parameterName=WPOIntensity, defaultValue=<rgba 0 7 0 1>}
  Add(
   A=LinearInterpolate(
    B=MaterialFunctionCall(WidthBase=@n1, WidthTip=@n1, Expand U or V UV Channel=@n2) {materialFunction=&SplineThicken}.WorldPositionOffset
    Alpha=ScalarParameter() {parameterName=UseSplineTickenSwitch}
   )
   B=Multiply(A=Power(Base=@n8) {constExponent=1}, B=Transform(None=@n9.RGB))
  )
  n10 = ParticleColor()
  n11 = TextureSampleParameter2D(UVs=NamedRerouteUsage()) {
   parameterName=AlphaTex
   samplerType=SAMPLERTYPE_Grayscale
   texture=/BRCosmetics/Weapons/FORT_Melee/Pickaxe_BentBaton/FX/Materials/T_BentBaton_Grayscale_Teeth05
  }
  n12 = TextureSampleParameter2D(UVs=NamedRerouteUsage()) {
   parameterName=AlphaTex
   samplerType=SAMPLERTYPE_Grayscale
   texture=&T_BentBaton_Grayscale_Teeth05
  }
  n13 = NamedRerouteUsage()
  n14 = TextureSampleParameter2D(UVs=Panner(Coordinate=TextureCoordinate() {uTiling=0.7}) {speedX=-0.3}) {
   parameterName=AlphaTex
   samplerType=SAMPLERTYPE_Grayscale
   texture=&T_BentBaton_Grayscale_Teeth05
  }
  n15 = TextureCoordinate() {uTiling=0.7}
  n16 = Constant2Vector() {r=0.4}
  n17 = TextureSampleParameter2D(UVs=Panner(Coordinate=Add(A=@n15, B=@n16)) {speedX=0.3}) {
   parameterName=AlphaTex
   samplerType=SAMPLERTYPE_Grayscale
   texture=&T_BentBaton_Grayscale_Teeth05
  }
  n18 = LinearInterpolate(A=@n14.RGB, B=@n17.RGB, Alpha=NamedRerouteUsage())
  n19 = NamedRerouteUsage()
  n20 = Saturate(None=Add(A=LinearInterpolate(A=@n11.RGB, B=@n12.RGB, Alpha=@n13), B=Multiply(A=@n18, B=@n19)))
  n21 = Multiply(
   A=@n10.A
   B=Power(
    Base=StaticComponentMaskParameter(None=@n20) {defaultG=true, parameterName=AlphaTexMask}
    Exp=ScalarParameter() {defaultValue=1, parameterName=Exp}
   ) {constExponent=1}
  )
  n22 = Sine(None=ComponentMask(None=TextureCoordinate()) {r=true}) {period=2}
  n23 = Saturate(None=Multiply(A=Power(Base=@n22)) {constB=7})
  n24 = MaterialFunctionCall("FadeSensitivity (optional)"=ScalarParameter() {parameterName=AngleFade}) {materialFunction=&FadeBasedOnViewAngle}
  n25 = ScalarParameter() {parameterName="Contrast (S)"}
  DepthFade(
   Opacity=Multiply(
    A=Multiply(A=@n21, B=@n23)
    B=MaterialFunctionCall(In=@n24.Result, Contrast=@n25) {materialFunction=&CheapContrast}.Result
   )
   FadeDistance=ScalarParameter() {parameterName=FadeDistance}
  )
  n26 = Panner(Coordinate=TextureCoordinate() {uTiling=0.5}) {speedX=-1}
  n27 = Multiply(
   A=LinearInterpolate(
    A=ScalarParameter() {defaultValue=-1, parameterName=DarkNoiseLerpMin}
    Alpha=TextureSample(UVs=@n26) {samplerType=SAMPLERTYPE_Grayscale, texture=&LowResBlurredNoise}.R
   ) {constA=-1, constB=5}
   B=VectorParameter() {defaultValue=<rgba 0.05 0 0 1>}.RGB
  )
  n28 = TextureSampleParameter2D(UVs=NamedRerouteUsage()) {
   parameterName=Lerp_TexGlow
   samplerType=SAMPLERTYPE_Grayscale
   texture=/BRCosmetics/Weapons/FORT_Melee/Pickaxe_BentBaton/FX/Materials/T_BentBaton_Grayscale_Teeth02
  }
  n29 = TextureSampleParameter2D(UVs=NamedRerouteUsage()) {
   parameterName=Lerp_TexGlow
   samplerType=SAMPLERTYPE_Grayscale
   texture=&T_BentBaton_Grayscale_Teeth02
  }
  n30 = TextureSampleParameter2D(UVs=Panner(Coordinate=TextureCoordinate() {uTiling=0.7}) {speedX=-0.3}) {
   parameterName=Lerp_TexGlow
   samplerType=SAMPLERTYPE_Grayscale
   texture=&T_BentBaton_Grayscale_Teeth02
  }
  n31 = TextureCoordinate() {uTiling=0.7}
  n32 = Constant2Vector() {r=0.4}
  n33 = TextureSampleParameter2D(UVs=Panner(Coordinate=Add(A=@n31, B=@n32)) {speedX=0.3}) {
   parameterName=Lerp_TexGlow
   samplerType=SAMPLERTYPE_Grayscale
   texture=&T_BentBaton_Grayscale_Teeth02
  }
  n34 = NamedRerouteUsage()
  n35 = Add(
   A=LinearInterpolate(A=@n28.RGB, B=@n29.RGB, Alpha=NamedRerouteUsage())
   B=Multiply(A=LinearInterpolate(A=@n30.RGB, B=@n33.RGB, Alpha=@n34), B=NamedRerouteUsage())
  )
  n36 = Multiply(
   A=VectorParameter() {parameterName=Param_1, defaultValue=<rgba 1 0.1406 0.1406 1>}.RGB
   B=StaticComponentMaskParameter(None=Saturate(None=@n35)) {defaultR=true, parameterName=LerpTexMask}
  )
  n37 = TextureCoordinate() {uTiling=0.5}
  n38 = NamedRerouteUsage()
  n39 = LinearInterpolate(Alpha=TextureSample(UVs=Panner(Coordinate=@n37, Time=@n38) {speedX=-0.5}) {samplerType=SAMPLERTYPE_Grayscale, texture=&LowResBlurredNoise}.R)
  n40 = VectorParameter() {parameterName=Color03, defaultValue=<rgba 0.932 0.0481 0.0481 1>}
  n41 = TextureSampleParameter2D(UVs=NamedRerouteUsage()) {
   parameterName=TexLight
   samplerType=SAMPLERTYPE_Grayscale
   texture=/BRCosmetics/Weapons/FORT_Melee/Pickaxe_BentBaton/FX/Materials/T_BentBaton_Grayscale_Teeth03
  }
  n42 = TextureSampleParameter2D(UVs=NamedRerouteUsage()) {
   parameterName=TexLight
   samplerType=SAMPLERTYPE_Grayscale
   texture=&T_BentBaton_Grayscale_Teeth03
  }
  n43 = NamedRerouteUsage()
  n44 = TextureSampleParameter2D(UVs=Panner(Coordinate=TextureCoordinate() {uTiling=0.7}) {speedX=-0.3}) {
   parameterName=TexLight
   samplerType=SAMPLERTYPE_Grayscale
   texture=&T_BentBaton_Grayscale_Teeth03
  }
  n45 = TextureCoordinate() {uTiling=0.7}
  n46 = Constant2Vector() {r=0.4}
  n47 = TextureSampleParameter2D(UVs=Panner(Coordinate=Add(A=@n45, B=@n46)) {speedX=0.3}) {
   parameterName=TexLight
   samplerType=SAMPLERTYPE_Grayscale
   texture=&T_BentBaton_Grayscale_Teeth03
  }
  n48 = LinearInterpolate(A=@n44.R, B=@n47.R, Alpha=NamedRerouteUsage())
  n49 = NamedRerouteUsage()
  n50 = Saturate(None=Add(A=LinearInterpolate(A=@n41.R, B=@n42.R, Alpha=@n43), B=Multiply(A=@n48, B=@n49)))
  n51 = Multiply(A=MaterialFunctionCall(In=Power(Base=@n50) {constExponent=1}, Contrast=Constant() {r=0.5}) {materialFunction=&CheapContrast}.Result)
  n52 = TextureSample(UVs=Panner(Coordinate=TextureCoordinate() {uTiling=0.5}, Time=NamedRerouteUsage()) {speedX=-0.5}) {samplerType=SAMPLERTYPE_Grayscale, texture=&LowResBlurredNoise}
  n53 = Add(A=Multiply(A=Multiply(A=@n40.RGB, B=@n51), B=LinearInterpolate(Alpha=@n52.R) {constB=3}), B=Add(A=Multiply(A=@n36, B=@n39)) {constB=0})
  n54 = TextureSampleParameter2D(UVs=NamedRerouteUsage()) {
   parameterName=TexCenter
   samplerType=SAMPLERTYPE_Grayscale
   texture=/BRCosmetics/Weapons/FORT_Melee/Pickaxe_BentBaton/FX/Materials/T_BentBaton_Grayscale_Teeth01
  }
  n55 = TextureSampleParameter2D(UVs=NamedRerouteUsage()) {
   parameterName=TexCenter
   samplerType=SAMPLERTYPE_Grayscale
   texture=&T_BentBaton_Grayscale_Teeth01
  }
  n56 = LinearInterpolate(A=@n54.R, B=@n55.R, Alpha=NamedRerouteUsage())
  n57 = TextureSampleParameter2D(UVs=Panner(Coordinate=TextureCoordinate() {uTiling=0.7}) {speedX=-0.3}) {
   parameterName=TexCenter
   samplerType=SAMPLERTYPE_Grayscale
   texture=&T_BentBaton_Grayscale_Teeth01
  }
  n58 = TextureCoordinate() {uTiling=0.7}
  n59 = Constant2Vector() {r=0.4}
  n60 = TextureSampleParameter2D(UVs=Panner(Coordinate=Add(A=@n58, B=@n59)) {speedX=0.3}) {
   parameterName=TexCenter
   samplerType=SAMPLERTYPE_Grayscale
   texture=&T_BentBaton_Grayscale_Teeth01
  }
  n61 = Multiply(A=LinearInterpolate(A=@n57.R, B=@n60.R, Alpha=NamedRerouteUsage()), B=NamedRerouteUsage())
  n62 = Power(Base=Saturate(None=Add(A=@n56, B=@n61))) {constExponent=10}
  n63 = TextureCoordinate() {uTiling=2}
  n64 = Multiply(
   A=LinearInterpolate(A=Add(A=@n53, B=@n27), B=VectorParameter() {parameterName=Param_2, defaultValue=<rgba 1 0.6479 0.6479 1>}.RGB, Alpha=Saturate(None=@n62))
   B=LinearInterpolate(Alpha=TextureSample(UVs=@n63) {samplerType=SAMPLERTYPE_Grayscale, texture=&LowResBlurredNoise}.R) {constA=0.8}
  )
  MaterialFunctionCall(Emissive=LinearInterpolate(A=Multiply(A=@n64, B=@n10.RGB), B=@n10.RGB, Alpha=ScalarParameter() {parameterName=UseSimpleGlow})) {materialFunction=&MF_FX_EmissiveAdjustments}
  n65 = TextureCoordinate() {uTiling=0.8}
  n66 = ComponentMask(None=VectorParameter() {parameterName=BaseUVScale, defaultValue=<rgba 1 1 0 0>}.RGB) {g=true, r=true}
  NamedRerouteDeclaration(None=Panner(Coordinate=Multiply(A=@n65, B=@n66)) {speedX=-0.3}) {name=Panner_R, nodeColor=<rgba 0 0.3385 1 1>}
  n67 = Time()
  n68 = ScalarParameter() {defaultValue=1, parameterName=Speed}
  n69 = Multiply(A=Constant() {r=2.5}, B=@n68)
  n70 = Constant() {r=8}
  n71 = Ceil(None=Multiply(A=Time() {bOverride_Period=true, period=6000}, B=Multiply(A=@n70, B=@n68)) {constB=8})
  NamedRerouteDeclaration(None=LinearInterpolate(A=Multiply(A=@n67, B=@n69) {constB=2.3}, B=Multiply(A=@n71) {constB=0.5}, Alpha=ScalarParameter() {defaultValue=0.4, parameterName=LerpStep}) {constAlpha=1}) {name=StepPanner, nodeColor=<rgba 1 0.0368 0.001 1>}
  n72 = Multiply(A=TextureCoordinate(), B=@n66)
  n73 = Constant2Vector() {r=0.5}
  NamedRerouteDeclaration(None=Panner(Coordinate=Add(A=@n72, B=@n73)) {speedX=0.3}) {name=Panner_L, nodeColor=<rgba 0 1 0.7188 1>}
  n74 = TextureCoordinate()
  n75 = SmoothStep(Value=OneMinus(None=ComponentMask(None=@n74) {r=true}))
  n76 = Constant() {r=5}
  NamedRerouteDeclaration(None=Multiply(A=MaterialFunctionCall(In=@n75, Contrast=@n76) {materialFunction=&CheapContrast}.Result)) {name=HalfMask, nodeColor=<rgba 1 0.0173 0.9793 1>}
  n77 = TextureCoordinate()
  n78 = Power(Base=Sine(None=ComponentMask(None=@n77) {r=true}) {period=2}) {constExponent=150}
  n79 = ScalarParameter() {defaultValue=1, parameterName=CentMaskRange}
  NamedRerouteDeclaration(None=Saturate(None=Multiply(A=@n78, B=@n79))) {name=CenterMask, nodeColor=<rgba 0.6172 1 0 1>}
  NamedRerouteUsage()
 ]
}
