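# Translucent, lit, two-sided Niagara mesh-particle material for smoke and dust spikes, widened through SplineThicken.
# Dynamic parameters drive erosion, panning, tiling and spline width, with per-particle random UV offsets.
# Panned cloud mask and normal textures combine with a streak mask for smoothstep-eroded opacity, softened by depth and camera-angle fades.
# Particle color takes quality-switched dust noise; optional EmissiveGlow adds HDR emissive, alongside a normal-based directional mask.
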Material /BeastGirlStormClawsGameplay/Assets/FX/Materials/M_Smoke_Dust_Spike
graph {
 properties {
  bTangentSpaceNormal=false
  bUsedWithNiagaraMeshParticles=true
  blendMode=BLEND_Translucent
  pixelDepthOffsetMode=PDOM_Legacy
  translucencyLightingMode=TLM_VolumetricPerVertexDirectional
  twoSided=true
  usedShadingModels=Default Lit
 }
 nodes [
  MaterialFunctionCall() {materialFunction=&DebugTimeSine}
  n1 = DynamicParameter() {defaultValue=<rgba 0 1 1 1>, paramNames [Erode,Pan,SplineWidth,Param4]}
  n2 = Multiply(A=@n1.SplineWidth, B=ScalarParameter() {defaultValue=100, parameterName=BaseSplineWidth})
  n3 = Add(A=TextureCoordinate(), B=ParticleRandom())
  n4 = DynamicParameter() {parameterIndex=1, defaultValue=<rgba 0.7 0.6 1 1>, paramNames [TileX,TileY,Param3,Param4]}
  n5 = NamedRerouteDeclaration(None=AppendVector(A=@n4.TileX, B=@n4.TileX)) {name=Tiling, nodeColor=<rgba 0 1 0.4375 1>}
  n6 = Constant()
  n7 = Multiply(A=ComponentMask(None=TextureCoordinate()) {g=true}, B=LinearInterpolate(Alpha=NamedRerouteUsage()) {constA=0.3, constB=-0.2}) {constB=0.3}
  n8 = Panner(Coordinate=Subtract(A=Multiply(A=@n3, B=@n5), B=AppendVector(A=@n6, B=@n7)), Time=NamedRerouteUsage()) {speedY=0.5}
  n9 = TextureSampleParameter2D(UVs=@n8) {
   parameterName=NormalTexture
   samplerType=SAMPLERTYPE_Normal
   texture=/Game/TimeOfDay/TODM_Sequence/Materials/Sky/Clouds/T_Asteria_Skybox_Clouds_N
  }
  n10 = MaterialFunctionCall(Normal=@n9.RGB, Flatness=LinearInterpolate(Alpha=NamedRerouteUsage()) {constA=0.3, constB=-1.5}) {materialFunction=&FlattenNormal}
  MaterialFunctionCall(
   WidthBase=Multiply(A=@n2, B=ScalarParameter() {defaultValue=0.2, parameterName=EndSplineWidthMult}) {constB=0.2}
   WidthTip=@n2
   AdditionalNormal=@n10.Result
  ) {materialFunction=&SplineThicken}
  n11 = Multiply(A=NamedRerouteUsage(), B=ParticleColor().RGB)
  n12 = QualitySwitch(
   Default=@n11
   High=MaterialFunctionCall(LDR Input=Saturate(None=@n11), HDR Tint=@n11, HDR Intensity Multiplier=ScalarParameter() {defaultValue=5, parameterName=HDRMult}) {materialFunction=&DeriveHDRfromLDR}.RemappedRange
  )
  n13 = Constant() {r=1}
  n14 = Add(A=Multiply(A=TextureCoordinate() {vTiling=0.1}, B=NamedRerouteUsage()), B=ParticleRandom())
  n15 = NamedRerouteUsage()
  n16 = Panner(Coordinate=Multiply(A=@n14) {constB=0.6}, Time=@n15) {speedX=-0.08, speedY=0.5}
  n17 = Panner(Coordinate=@n14, Time=@n15) {speedX=0.15, speedY=0.7}
  n18 = Add(A=Max(
    A=TextureSample(UVs=@n16) {texture=/Game/Global/Textures/Default/PointSplatter/T_Tiling_Dust}.RGB
    B=TextureSample(UVs=@n17) {texture=&T_Tiling_Dust}.RGB
   )) {constB=0.9}
  n19 = Multiply(A=@n18, B=@n18)
  n20 = QualitySwitch(Default=@n13, High=@n19, Epic=@n19)
  n21 = ParticleColor()
  n22 = StaticSwitchParameter(
   True=VectorParameter() {parameterName=BaseColor, defaultValue=<rgba 0.12 0.12 0.12 1>}.RGB
   False=Multiply(A=StaticSwitchParameter(True=@n20, False=@n13) {defaultValue=true, parameterName=UseDustNoise}, B=@n21.RGB)
  ) {parameterName=EmissiveGlow}
  n23 = Multiply(A=ScalarParameter() {defaultValue=0.2, parameterName=BaseColorEmissiveMult}, B=@n22) {constB=0.01}
  MaterialFunctionCall(Emissive=StaticSwitchParameter(True=Add(A=@n12, B=@n23), False=@n23) {parameterName=EmissiveGlow}) {materialFunction=&MF_FX_EmissiveAdjustments}
  n24 = TextureCoordinate() {uTiling=0.4}
  n25 = AppendVector(
   A=MaterialFunctionCall(Float2=NamedRerouteUsage()) {materialFunction=&BreakOutFloat2Components}.R
   B=Constant() {r=1}
  )
  n26 = Multiply(A=ParticleRandom()) {constB=1.2}
  n27 = Constant()
  n28 = TextureSample(UVs=Add(A=Multiply(A=@n24, B=@n25), B=AppendVector(A=@n26, B=@n27))) {samplerType=SAMPLERTYPE_Grayscale, texture=/Game/Effects/Fort_Effects/Textures/Masks/T_Vertical_Spike_Streak_M}
  n29 = TextureSampleParameter2D(UVs=@n8) {
   parameterName=MaskTexture
   samplerType=SAMPLERTYPE_LinearGrayscale
   texture=/Game/TimeOfDay/TODM_Sequence/Materials/Sky/Clouds/T_Asteria_Skybox_Clouds
  }
  n30 = Multiply(A=@n28.G, B=@n29.G)
  Multiply(
   A=@n22
   B=NamedRerouteDeclaration(None=LinearInterpolate(Alpha=@n30) {constA=0.4, constB=0.8}) {name=TextureShading, nodeColor=<rgba 0.9922 0 1 1>}
  )
  n31 = TextureCoordinate()
  n32 = Constant() {r=0.5}
  n33 = Multiply(A=SphereMask(A=ComponentMask(None=@n31) {r=true}, B=@n32, Radius=@n32) {attenuationRadius=0.5, hardnessPercent=15}, B=@n28.G)
  n34 = NamedRerouteUsage()
  n35 = LinearInterpolate(
   A=ScalarParameter() {defaultValue=-0.05, parameterName=ErodeSoftnessStart}
   B=ScalarParameter() {defaultValue=-0.5, parameterName=ErodeSoftnessEnd}
   Alpha=@n34
  )
  n36 = SmoothStep(Min=Add(A=@n35, B=@n34) {constB=-0.15}, Max=@n34, Value=Add(A=@n29.G, B=@n35) {constB=-0.15})
  n37 = Multiply(A=@n21.A, B=Multiply(A=Multiply(A=@n33, B=@n36)) {constB=1.5})
  n38 = ComponentMask(None=TextureCoordinate()) {g=true}
  n39 = DepthFade(
   Opacity=@n37
   FadeDistance=Multiply(A=LinearInterpolate(Alpha=@n38) {constA=1, constB=0.1}, B=ScalarParameter() {defaultValue=100, parameterName=DepthFade})
  )
  n40 = Add(A=DotProduct(A=MaterialFunctionCall() {materialFunction=&TangentBasis}.Y V, B=CameraVectorWS()))
  Multiply(
   A=ShadingPathSwitch(Default=QualitySwitch(Default=@n39, Low=@n37), Deferred=@n39)
   B=Power(Base=Saturate(None=@n40), Exp=ScalarParameter() {defaultValue=0.35, parameterName=CameraAngleFadePower}) {constExponent=0.5}
  )
  NamedRerouteDeclaration(None=@n1.Erode) {name=Anim, nodeColor=<rgba 0.4531 0 1 1>}
  NamedRerouteDeclaration(None=@n4.TileX) {name=TileX, nodeColor=<rgba 1 0.1641 0 1>}
  NamedRerouteDeclaration(None=@n1.Pan) {name=Pan, nodeColor=<rgba 1 0.2109 0 1>}
  n41 = ComponentMask(None=TextureCoordinate()) {g=true}
  n42 = Multiply(A=ScalarParameter() {defaultValue=0.75, parameterName=EmissiveMaskGradientMult}, B=Multiply(A=@n41, B=@n41))
  n43 = Saturate(None=DotProduct(A=VectorParameter() {parameterName=EmissiveDirection, defaultValue=<rgba 0 0.7 0.3 1>}.RGB, B=@n10.Result))
  NamedRerouteDeclaration(None=Power(Base=LinearInterpolate(A=@n42, Alpha=@n43), Exp=ScalarParameter() {defaultValue=4, parameterName=EmissiveDirectionPower})) {name=EmissiveMask, nodeColor=<rgba 0.5 0 1 1>}
  Power(Base=@n7)
 ]
}
