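# Niagara system for BeastGirl StormClaws that draws a red, electrified ground ring indicator sized by User.Ring Scale.
# A flickering ribbon edge, widened with player distance, sits over a decal with ring-spawned sparks and electricity meshes and sprites.
# Most spawn rates scale with radius, and sparks enlarge with LOD distance.
# An UpdateValueStateChange flag, negated and skipped on the first frame, gates the GPU spark burst and timeline-driven brightness flashes.
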NiagaraSystem /BeastGirlStormClawsGameplay/Assets/FX/Systems/NS_BeastGirl_Indicator
dependencies {
 usedDataInterfaces [
  &NiagaraDataInterfaceArrayFloat
  &NiagaraDataInterfaceCurve
  &NiagaraDataInterfaceColorCurve
  &NiagaraDataInterfaceSpriteRendererInfo
 ]
 usedDynamicInputs [&MakeFloatFromVector]
 usedModules [&SystemState, /Game/Effects/Niagara/Modules/Logic/UpdateValueStateChange]
}
emitters {
 Burst {
  simTarget=GPUComputeSim
  emitterUpdateScript [
   &EmitterState()
   &SpawnPerFrame(
    Spawn:NiagaraBool=&Boolean_AndOperation(
     Bool1=&Boolean_NotOperation(Bool=$System.UpdateValueStateChange.Changed)
     Bool2=&Boolean_NotOperation(Bool=$Transient.FirstFrame)
    )
    Spawn Count=500
    Use Spawn Probability=false
   )
   &SpawnRate(
    SpawnRate=&Multiply_Float(
     A=1500
     B=&RemapRange(Input Value To Remap=$User.Ring Scale, Input Min=0, Input Max=1500, Output Min=0, Output Max=1)
    )
    Use Spawn Probability=false
   ) {enabled=false}
  ]
  particleSpawnScript [
   &InitializeParticle(
    Color=<rgba 10 0 0.1667 1>
    Lifetime Max=0.75
    Lifetime Min=0.25
    Lifetime Mode=::NewEnumerator1
    Position Offset=<0 0 12>
    Sprite Size Mode=::NewEnumerator1
    Uniform Sprite Size Max=5
    Uniform Sprite Size Min=3
    UsePositionOffset=true
    Write Material Random=true
   )
   &ShapeLocation(
    MaskSpawnGroup=false
    Offset=<0 0 30>
    Offset Mode=::NewEnumerator0
    Override Seed=false
    Ring Radius=&Multiply_Float(A=$User.Ring Scale, B=&RandomRangeFloat(Minimum=0.99, Maximum=1.02))
    Shape Primitive=::NewEnumerator4
    UseOwnerScale=true
   )
   &AddVelocity(
    Velocity=&AddVectorToPosition(Position=&Multiply_Vector(A=$Particles.ShapeLocation.ShapeTangent, B=<1 1 1>), Vector=<0 0 0.5>)
    Velocity Speed Scale=&Multiply_Float(A=&RandomRangeFloat(Minimum=0, Maximum=1500), B=1)
   )
   &CameraOffset(Camera Offset Amount=50)
   &InheritVelocity(Inherited Velocity Amount Scale=<-0.05 -0.05 -0.05>, Speed Limit=false, UseSpeedThreshold=false)
  ]
  particleUpdateScript [
   &ParticleState()
   &ScaleSpriteSize()
   &VortexForce(
    Origin Pull Amount=50
    UseOriginOffset=false
    Vortex Force Amount=&RandomRangeFloat(Minimum=0, Maximum=750)
   )
   &Drag(Drag=4, Use Linear Drag=true, Use Rotational Drag=true)
   &GravityForce(Gravity=&Multiply_VectorByFloat(Vector=<0 0 -980>, Float=&RandomRangeFloat(Minimum=1, Maximum=2)))
   &SolveForcesAndVelocity(Clamp Velocity=false, Limit Acceleration=false)
   &ScaleSpriteSizeBySpeed(Max Scale Factor=<1 4>, Min Scale Factor=<0.5 0.5>)
   &ScaleColor(
    Linear Color Curve=DataInterface() {propertyValues {
      AlphaCurve {keys [<at 0 1>,<at 1 0>]}
      BlueCurve {keys [<at 0 15>,<at 1 1>]}
      GreenCurve {keys [<at 0 15>,<at 1 1>]}
      RedCurve {keys [<at 0 15>,<at 1 1>]}
     }}
    Scale Mode=::NewEnumerator2
   )
   &ScaleSpriteSize(
    Scale Sprite Size Mode=::NewEnumerator0
    Uniform Scale Factor=&Lerp_Float(
     A=1
     B=5
     Alpha=&FN_LOD_Distance_Remap(Start Distance=0, End Distance=3500, Minimum Output=0, Maximum Output=1, Clamp Output=true)
    )
   ) {module=ScaleSpriteSize001}
  ]
  rendererClasses [&NiagaraSpriteRendererProperties]
  renderers [{rendererClass=&NiagaraSpriteRendererProperties}]
 }
 Decal {
  emitterUpdateScript [
   &EmitterState(Life Cycle Mode=::NewEnumerator1, Loop Duration=1, UseLoopDelay=false)
   &SpawnBurst_Instantaneous(Use Spawn Probability=false, UseLoopCountLimit=false)
   &Timeline(
    Additional Output:ENiagara_TimelineOutput=::NewEnumerator1
    Additional Output Mode:ENiagara_TimelineOutputMode=::NewEnumerator1
    Float Curve=DataInterface() {propertyValues {Curve {keys [
        {interpMode=RCIM_Cubic, tangentMode=RCTM_User, time=0, value=0}
        {interpMode=RCIM_Cubic, tangentMode=RCTM_User, time=0.5, value=1}
        {tangentMode=RCTM_User, time=1, value=0}
       ]}}}
    Play=&Boolean_AndOperation(
     Bool1=&Boolean_NotOperation(Bool=$System.UpdateValueStateChange.Changed)
     Bool2=&Boolean_NotOperation(Bool=$Transient.FirstFrame)
    )
    Play Rate=20
    Rewind Rate=2
   )
   /SetVariables() {isSetParameters=true}
  ]
  particleSpawnScript [
   &InitializeParticle(Color=<rgba 0.25 0 0 1>, Sprite Size Mode=::NewEnumerator3, Uniform Sprite Size=2, UsePositionOffset=false, Write Material Random=true)
   /Niagara/Modules/Spawn/Decal/InitialDecalOrientation(Decal Axis to Orient:Vector3f=<0 0 -1>, Use Rotation Vector=false)
   &SpriteFacingAndAlignment(Sprite Facing=<0 0 1>, WriteAlignment=false, WriteFacing=true)
  ]
  particleUpdateScript [
   /SetVariables(Particles.DecalSize:Vector3f=&Multiply_VectorByFloat(Vector=<2 2 2>, Float=$User.Ring Scale)) {isSetParameters=true}
   &ParticleState()
   &ScaleSpriteSize(Scale Sprite Size Mode=::NewEnumerator0, Uniform Scale Factor=$User.Ring Scale)
   &ScaleColor(Scale RGB=&Lerp_Vector(A=<1 1 1>, B=<4 4 4>, Alpha=$Emitter.Timeline.Float), ScaleA=true, ScaleRGB=true)
   /SetVariables() {enabled=false, isSetParameters=true}
  ]
  rendererClasses [&NiagaraDecalRendererProperties, &NiagaraSpriteRendererProperties]
  renderers [
   {rendererClass=&NiagaraDecalRendererProperties, rendererIndex=0}
   {rendererClass=&NiagaraSpriteRendererProperties, rendererIndex=1}
  ]
 }
 Electricty {
  emitterUpdateScript [
   &EmitterState()
   &SpawnRate(
    SpawnRate=&Multiply_Float(
     A=30
     B=&RemapRange(Input Value To Remap=$User.Ring Scale, Input Min=0, Input Max=1500, Output Min=0, Output Max=2)
    )
    Use Spawn Probability=false
   )
  ]
  particleSpawnScript [
   &PartitionParticles()
   &InitializeParticle(
    Color=<rgba 1 0 0.0167 1>
    Lifetime Max=0.35
    Lifetime Min=0.25
    Lifetime Mode=::NewEnumerator1
    Mesh Scale=&Multiply_VectorByFloat(Vector=&RandomRangeVector(Minimum=<1 2 0.5>, Maximum=<2 2 3>), Float=1)
    Mesh Scale Mode=::NewEnumerator4
    Position Offset=<0 0 12>
    Sprite Size Mode=::NewEnumerator3
    Uniform Sprite Size=350
    UsePositionOffset=true
    Write Material Random=true
   )
   &ShapeLocation(
    MaskSpawnGroup=false
    Offset Mode=::NewEnumerator0
    Override Seed=false
    Ring Radius=&Multiply_Float(A=$User.Ring Scale, B=&RandomRangeFloat(Minimum=0.99, Maximum=1.01))
    Shape Primitive=::NewEnumerator4
    UseOwnerScale=true
   )
   &InitialMeshOrientation(
    Mesh Axis To Orient=<-1 0 0>
    Mesh Orientation Mode=::NewEnumerator2
    Orientation Vector=$Particles.ShapeLocation.ShapeBitangent
    Use Rotation Vector=false
   )
   &AddVelocity(
    Velocity=$Particles.ShapeLocation.ShapeBitangent
    Velocity Speed Scale=&RandomRangeFloat(Minimum=0, Maximum=-50)
   )
  ]
  particleUpdateScript [
   &ParticleState()
   &ScaleSpriteSize(Uniform Curve Sprite Scale=DataInterface() {propertyValues {Curve {keys [
        {interpMode=RCIM_Cubic, leaveTangent=5.8794, tangentMode=RCTM_Break, time=0, value=0}
        {interpMode=RCIM_Cubic, tangentMode=RCTM_Break, time=0.5, value=1.05}
        {arriveTangent=-6.7193, interpMode=RCIM_Cubic, tangentMode=RCTM_Break, time=1, value=0}
       ]}}})
   &ScaleMeshSize(Scale Factor=&MakeVector(
     X=&FloatFromCurve(FloatCurve=DataInterface() {propertyValues {Curve {
         defaultValue=1
         keys [{interpMode=RCIM_Cubic, leaveTangent=2.7575, tangentMode=RCTM_Break, time=0.25, value=0}, {tangentMode=RCTM_User, time=1, value=1.375}]
        }}})
     Y=1
     Z=&FloatFromCurve(FloatCurve=DataInterface() {propertyValues {Curve {
         defaultValue=1
         keys [{interpMode=RCIM_Cubic, leaveTangent=2.2978, tangentMode=RCTM_Break, time=0, value=0}, {tangentMode=RCTM_User, time=1, value=1}]
        }}})
    ))
   &DynamicMaterialParameters(
    Index 0 Param 1=&Multiply_Float(
     A=&Lerp_Float(
      A=0.5
      B=0.75
      Alpha=&FN_LOD_Distance_Remap(Start Distance=0, End Distance=3500, Minimum Output=0, Maximum Output=1, Clamp Output=true)
     )
     B=&VectorLength(Vector=$Engine.Owner.Scale)
    )
    Index 0 Param 2=&Multiply_Float(A=0.25, B=&VectorLength(Vector=$Engine.Owner.Scale))
    Index 0 Param 3=&RandomRangeFloat(Minimum=0, Maximum=1)
    Index 0 Param 4=&RandomRangeFloat(Minimum=0, Maximum=1)
    Module.Index0Float0Write=true
    Module.Index0Float1Write=true
    Module.Index0Float2Write=true
    Module.Index0Float3Write=true
   )
   &SolveForcesAndVelocity(Clamp Velocity=false, Limit Acceleration=false)
   /SetVariables(Particles.DynamicParamOverride:Vector4f=&MakeFloat4FromVectorAndFloat(
     "Vector (RGB)"=/Niagara/DynamicInputs/Transforms/TransformPosition(
      Position=Linked() {linkedVariable {name=Particles.Initial.Position, type=NiagaraPosition}}
      Source Space=::Simulation
      Destination Space=::World
      Apply Scale=true
     )
     "Float (Alpha)"=0
    )) {isSetParameters=true}
   &ScaleColor(
    Scale Alpha=&FloatFromCurve(FloatCurve=DataInterface() {propertyValues {Curve {keys [{interpMode=RCIM_Cubic, tangentMode=RCTM_Break, time=0, value=1}, <at 1 0>]}}})
    ScaleA=true
    ScaleRGB=true
   ) {module=ScaleColor001}
   &FN_Sprite_Attributes(
    Camera Offset=20
    Color=&MultiplyLinearColor(Linear Color=$Particles.Color, Scale Factor=<1 1 1 0.05>)
    Create Camera Offset Attribute=true
    Create Color Attribute=true
    Write to Module Namespace=true
   )
  ]
  rendererClasses [&NiagaraMeshRendererProperties, &NiagaraSpriteRendererProperties]
  renderers [
   {rendererClass=&NiagaraMeshRendererProperties, rendererIndex=0}
   {rendererClass=&NiagaraSpriteRendererProperties, rendererIndex=1}
  ]
 }
 Minimal002 {
  emitterUpdateScript [
   &EmitterState()
   &SpawnRate(
    SpawnRate=&Multiply_Float(
     A=100
     B=&RemapRange(Input Value To Remap=$User.Ring Scale, Input Min=0, Input Max=1500, Output Min=0, Output Max=5)
    )
    Use Spawn Probability=false
   )
  ]
  particleSpawnScript [
   &PartitionParticles()
   &InitializeParticle(
    Color=<rgba 2 0 0.0333 0.5>
    Lifetime Max=0.25
    Lifetime Min=0.17
    Lifetime Mode=::NewEnumerator1
    Position Offset=<0 0 5>
    Sprite Rotation Mode=::NewEnumerator2
    Sprite Size Mode=::NewEnumerator1
    Sprite UV Mode=::NewEnumerator3
    Uniform Sprite Size Max=100
    Uniform Sprite Size Min=75
    UsePositionOffset=true
    Write Material Random=true
   )
   &ShapeLocation(
    MaskSpawnGroup=false
    Offset=<0 0 30>
    Offset Mode=::NewEnumerator0
    Override Seed=false
    Ring Radius=&Multiply_Float(A=$User.Ring Scale, B=&RandomRangeFloat(Minimum=0.99, Maximum=1.01))
    Shape Primitive=::NewEnumerator4
    UseOwnerScale=true
   )
   &InitialMeshOrientation(
    Mesh Axis To Orient=<-1 0 0>
    Mesh Orientation Mode=::NewEnumerator2
    Orientation Vector=$Particles.ShapeLocation.ShapeBitangent
    Use Rotation Vector=false
   )
   &AddVelocity(Velocity=$Particles.ShapeLocation.ShapeBitangent, Velocity Speed Scale=50)
   &SpriteFacingAndAlignment(Sprite Facing=<0 0 1>, WriteAlignment=false, WriteFacing=true)
   &SubUVAnimation(
    "Sprite Renderer "=DataInterface() {propertyValues {SpriteRenderer=&self:Minimal002_0.NiagaraSpriteRendererProperties_1}}
    SubUV Animation Mode:ENiagara_SubUVLookupModeV2=::NewEnumerator2
    UseEndFrame=false
    UseStartFrame=false
   )
  ]
  particleUpdateScript [
   &ParticleState()
   &ScaleSpriteSize()
   &ScaleSpriteSize(
    Scale Sprite Size Mode=::NewEnumerator0
    Uniform Scale Factor=&Lerp_Float(A=1, B=1.5, Alpha=&VectorLength(Vector=$Engine.Owner.Scale))
   ) {module=ScaleSpriteSize001}
   &DynamicMaterialParameters(
    Index 0 Param 1=&Multiply_Float(
     A=&Lerp_Float(
      A=0.1
      B=0.5
      Alpha=&FN_LOD_Distance_Remap(Start Distance=0, End Distance=3500, Minimum Output=0, Maximum Output=1, Clamp Output=true)
     )
     B=&VectorLength(Vector=$Engine.Owner.Scale)
    )
    Index 0 Param 2=&Multiply_Float(A=0.5, B=&VectorLength(Vector=$Engine.Owner.Scale))
    Index 0 Param 3=&RandomRangeFloat(Minimum=0, Maximum=1)
    Index 0 Param 4=&RandomRangeFloat(Minimum=0, Maximum=1)
    Module.Index0Float0Write=true
    Module.Index0Float1Write=true
    Module.Index0Float2Write=true
    Module.Index0Float3Write=true
    Module.Index1Float0Write=true
    Module.Index1Float1Write=true
    Module.Index1Float2Write=true
    Module.Index1Float3Write=true
    Write Parameter Index 1=::NewEnumerator1
   )
   &SolveForcesAndVelocity(Clamp Velocity=false, Limit Acceleration=false)
  ]
  rendererClasses [&NiagaraSpriteRendererProperties]
  renderers [{rendererClass=&NiagaraSpriteRendererProperties}]
 }
 Sparks {
  simTarget=GPUComputeSim
  emitterUpdateScript [
   &EmitterState()
   &SpawnRate(
    SpawnRate=&Multiply_Float(
     A=1500
     B=&RemapRange(Input Value To Remap=$User.Ring Scale, Input Min=0, Input Max=1500, Output Min=0, Output Max=1)
    )
    Use Spawn Probability=false
   )
  ]
  particleSpawnScript [
   &InitializeParticle(
    Color=<rgba 10 0 0.1667 1>
    Lifetime Max=0.5
    Lifetime Min=0.25
    Lifetime Mode=::NewEnumerator1
    Position Offset=<0 0 12>
    Sprite Size Mode=::NewEnumerator1
    Uniform Sprite Size Max=5
    Uniform Sprite Size Min=3
    UsePositionOffset=true
    Write Material Random=true
   )
   &ShapeLocation(
    MaskSpawnGroup=false
    Offset=<0 0 30>
    Offset Mode=::NewEnumerator0
    Override Seed=false
    Ring Radius=&Multiply_Float(A=$User.Ring Scale, B=&RandomRangeFloat(Minimum=0.99, Maximum=1.02))
    Shape Primitive=::NewEnumerator4
    UseOwnerScale=true
   )
   &AddVelocity(
    Velocity=&AddVectorToPosition(Position=&Multiply_Vector(A=$Particles.ShapeLocation.ShapeBitangent, B=<1 1 1>), Vector=<0 0 0>)
    Velocity Speed Scale=&Multiply_Float(
     A=&RandomRangeFloat(Minimum=-250, Maximum=-750)
     B=&RemapRange(Input Value To Remap=$User.Ring Scale, Input Min=0, Input Max=1500, Output Min=1, Output Max=1.5)
    )
   )
   &CameraOffset(Camera Offset Amount=50)
   &InheritVelocity(Inherited Velocity Amount Scale=<-0.05 -0.05 -0.05>, Speed Limit=false, UseSpeedThreshold=false)
  ]
  particleUpdateScript [
   &ParticleState()
   &ScaleSpriteSize()
   &VortexForce(Origin Pull Amount=50, UseOriginOffset=false)
   &Drag(Drag=4, Use Linear Drag=true, Use Rotational Drag=true)
   &GravityForce(Gravity=&Multiply_VectorByFloat(Vector=<0 0 -980>, Float=&RandomRangeFloat(Minimum=1, Maximum=2)))
   &SolveForcesAndVelocity(Clamp Velocity=false, Limit Acceleration=false)
   &ScaleSpriteSizeBySpeed(Max Scale Factor=<1 4>, Min Scale Factor=<0.5 0.5>)
   &ScaleColor(
    Linear Color Curve=DataInterface() {propertyValues {
      AlphaCurve {keys [<at 0 1>,<at 1 0>]}
      BlueCurve {keys [<at 0 5>,<at 1 1>]}
      GreenCurve {keys [<at 0 5>,<at 1 1>]}
      RedCurve {keys [<at 0 5>,<at 1 1>]}
     }}
    Scale Mode=::NewEnumerator2
   )
   &ScaleSpriteSize(
    Scale Sprite Size Mode=::NewEnumerator0
    Uniform Scale Factor=&Lerp_Float(
     A=1
     B=5
     Alpha=&FN_LOD_Distance_Remap(Start Distance=0, End Distance=3500, Minimum Output=0, Maximum Output=1, Clamp Output=true)
    )
   ) {module=ScaleSpriteSize001}
  ]
  rendererClasses [&NiagaraSpriteRendererProperties]
  renderers [{rendererClass=&NiagaraSpriteRendererProperties}]
 }
 TheEdgeRing {
  emitterUpdateScript [
   &EmitterState(
    Inactive Response=::NewEnumerator1
    Life Cycle Mode=::NewEnumerator1
    Loop Behavior=::NewEnumerator1
    Loop Duration Mode=::NewEnumerator1
    UseLoopDelay=false
   )
   &Timeline(
    Additional Output:ENiagara_TimelineOutput=::NewEnumerator1
    Additional Output Mode:ENiagara_TimelineOutputMode=::NewEnumerator1
    Play=&Boolean_AndOperation(
     Bool1=&Boolean_NotOperation(Bool=$System.UpdateValueStateChange.Changed)
     Bool2=&Boolean_NotOperation(Bool=$Transient.FirstFrame)
    )
    Play Rate=20
    Rewind Rate=2
   )
   &SpawnBurst_Instantaneous(Spawn Count=24, Use Spawn Probability=false, UseLoopCountLimit=false)
  ]
  particleSpawnScript [
   &PartitionParticles(Number Of Partitions=3, Partition Mode:ENiagara_PartitionMode_v2=::NewEnumerator1)
   &InitializeParticle(
    Color=<rgba 10 0 0.1667 1>
    Lifetime=0.75
    Mesh Scale Mode=::NewEnumerator3
    Mesh Uniform Scale=15.5745
    Ribbon Facing Mode=::NewEnumerator1
    Ribbon Width=7
    Ribbon Width Mode=::NewEnumerator1
    UsePositionOffset=true
    Write Material Random=true
   )
   &InitialMeshOrientation(Mesh Axis To Orient=<0 0 1>, Mesh Orientation Mode=::NewEnumerator2, Use Rotation Vector=false)
  ]
  particleUpdateScript [
   &ParticleState(Kill Particles When Lifetime Has Elapsed=false, Let Infinitely Lived Particles Die When Emitter Deactivates=true)
   &ShapeLocation(
    MaskSpawnGroup=false
    Offset=<0 0 30>
    Offset Mode=::NewEnumerator0
    Override Seed=false
    Ring / Disc Distribution Mode:ENiagaraSphereDistributionMode=::NewEnumerator1
    Ring Radius=&Multiply_Float(A=$User.Ring Scale, B=1)
    Shape Primitive=::NewEnumerator4
    U Position=&ReturnNormalizedExecIndex()
    UseOwnerScale=true
   )
   &CameraOffset(Camera Offset Amount=10)
   &SolveForcesAndVelocity(Clamp Velocity=false, Limit Acceleration=false)
   &DynamicMaterialParameters(Module.Param0WriteEnabled=true, Module.Param1WriteEnabled=true, Module.Param2WriteEnabled=true, Module.Param3WriteEnabled=true)
   &ScaleColor(
    Scale RGB=&VectorFromFloat(Value=&Waveform("Waveform [1]"=::NewEnumerator8, "[1] Frequency"=25, "[1] Amplitude Min/Max"=<1 3>))
    ScaleA=true
    ScaleRGB=true
   )
   &ScaleRibbonWidth(Ribbon Width Scale=&Multiply_Float(
     A=&RemapRange(Input Value To Remap=$User.Ring Scale, Input Min=0, Input Max=1500, Output Min=1, Output Max=3)
     B=&Lerp_Float(A=1.5, B=2, Alpha=$Emitter.Timeline.Float)
    ))
   &ScaleColor(
    Linear Color Curve=DataInterface() {propertyValues {
      AlphaCurve {keys [<at 0 1>]}
      BlueCurve {keys [<at 0 0.1>,<at 0.1696 9.0082>,<at 1 1>]}
      GreenCurve {keys [<at 0 0.1>,<at 0.1696 16>,<at 1 1>]}
      RedCurve {keys [<at 0 0.1>,<at 0.1696 16>,<at 1 1>]}
     }}
    Scale Mode=::NewEnumerator2
   ) {module=ScaleColor001}
   &ScaleRibbonWidth(Ribbon Width Scale=&Multiply_Float(
     A=&Waveform("Waveform [1]"=::NewEnumerator8, "[1] Frequency"=20, "[1] Amplitude Min/Max"=<0.5 1>, "[1] Bias"=0.25)
     B=&Lerp_Float(
      A=1
      B=4
      Alpha=&RemapRange(
       Input Value To Remap=&VectorLength(Vector=&Subtract_Vector(
         A=/Niagara/DynamicInputs/Transforms/Position/TransformPosition(
          Position=&MakeVectorFromLinearColorRGB(Color=$NPC.FortniteNPC.PlayerPosition)
          Source Space=::World
          Destination Space=::Local
          Apply Scale=true
         )
         B=Linked() {linkedVariable {name=Particles.Position, type=NiagaraPosition}}
        ))
       Input Min=0
       Input Max=7500
       Output Min=0
       Output Max=1
      )
     )
    )) {module=ScaleRibbonWidth001}
   &ScaleColor(
    Scale RGB=&Lerp_Vector(
     A=<1 1 1>
     B=<5 5 5>
     Alpha=&RemapRange(Input Value To Remap=$User.Ring Scale, Input Min=0, Input Max=1500, Output Min=0, Output Max=1)
    )
    ScaleA=true
    ScaleRGB=true
   ) {module=ScaleColor003}
   &ScaleColor(Scale RGB=&Lerp_Vector(A=<1 1 1>, B=<50 50 50>, Alpha=$Emitter.Timeline.Float), ScaleA=true, ScaleRGB=true) {module=ScaleColor002}
   /SetVariables(Particles.RibbonWidth2:NiagaraFloat=&Multiply_Float(A=$Particles.RibbonWidth, B=2)) {isSetParameters=true}
  ]
  rendererClasses [&NiagaraRibbonRendererProperties]
  renderers [
   {rendererClass=&NiagaraRibbonRendererProperties}
   {rendererClass=&NiagaraRibbonRendererProperties, rendererIndex=1}
  ]
 }
}
summary {
 systemName=NS_BeastGirl_Indicator
 emitters [
  {emitterName=TheEdgeRing, rendererClasses [&NiagaraRibbonRendererProperties]}
  {emitterName=Sparks, simTarget=GPUComputeSim, rendererClasses [&NiagaraSpriteRendererProperties]}
  {emitterName=Electricty, rendererClasses [&NiagaraMeshRendererProperties, &NiagaraSpriteRendererProperties]}
  {emitterName=Minimal002, rendererClasses [&NiagaraSpriteRendererProperties]}
  {emitterName=Decal, rendererClasses [&NiagaraDecalRendererProperties, &NiagaraSpriteRendererProperties]}
  {emitterName=Burst, simTarget=GPUComputeSim, rendererClasses [&NiagaraSpriteRendererProperties]}
 ]
 userVariables [
  User.Burst Sparks {type=NiagaraBool}
  User.Level {type=NiagaraInt32, defaultValue=1}
  User.Ring Scale {type=NiagaraFloat, defaultValue=1500, description=""}
 ]
}
