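# Niagara system for a BeastGirl lightning-strike transformation effect, judging by its name.
# It exposes a TargetEnd position user parameter, which suggests an externally supplied endpoint for the strike.
# At system level it uses the SystemState module and a Set Variables assignment in the spawn script.
# The summary lists no emitters, data interfaces or dynamic inputs, so no visual elements appear here.
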NiagaraSystem /BeastGirlStormClawsGameplay/Assets/FX/Systems/NS_BeastGirl_Transform_Lightning_Strike_Transformation
dependencies {
 usedDataInterfaces []
 usedDynamicInputs []
 usedModules [
  &SystemState
  &self:SystemSpawnScript.SystemScriptSource.SystemScriptGraph.NiagaraNodeAssignment_5.SetVariables_#1
 ]
}
summary {
 systemName=NS_BeastGirl_Transform_Lightning_Strike_Transformation
 emitters []
 userVariables [User.TargetEnd {type=NiagaraPosition, description="", defaultValue {}}]
}
