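# Niagara system for a landing effect, named for a BentBaton jump landing.
# It holds a single Dust_Kickup emitter, and SystemState is the only module listed.
# No renderers, data interfaces, dynamic inputs or user variables are listed, so the dust's look and behavior are not shown.
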NiagaraSystem /BeastGirlStormClawsGameplay/Assets/FX/Systems/NS_BentBaton_Jump_Landing
dependencies {usedDataInterfaces [], usedDynamicInputs [], usedModules [&SystemState]}
emitters {Dust_Kickup {rendererClasses [], renderers []}}
summary {systemName=NS_BentBaton_Jump_Landing, emitters [{emitterName=Dust_Kickup, rendererClasses []}], userVariables []}
