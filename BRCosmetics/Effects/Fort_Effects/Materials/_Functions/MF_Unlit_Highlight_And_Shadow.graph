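# Material function for unlit effects that computes time-of-day highlight, shadow and light-through-smoke colors from Fortnite's global parameter collection.
# Highlights use a powered normal-to-sun dot product tinted by sun color and scaled by sun height, then add to shadows built from fog color.
# A second output multiplies a flattened-normal Fresnel by a custom Schlick phase function, sun and fog inscattering color, and a brightness input.
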MaterialFunction /BRCosmetics/Effects/Fort_Effects/Materials/_Functions/MF_Unlit_Highlight_And_Shadow
graph {
 properties {bExposeToLibrary=true, description=Color Based on Time of Day, libraryCategoriesText [Misc,FX_Common]}
 nodes [
  n1 = ScalarParameter() {defaultValue=0.25, parameterName=LightThroughSmokeBrightness}
  n2 = FunctionInput(Preview=Transform(None=TextureSample() {samplerType=SAMPLERTYPE_Normal, texture=/Game/Global/Textures/Default/Tileable_Patterns/T_StormCloudDetail01_N}.RGB)) {bUsePreviewValueAsDefault=true, inputName=WorldSpaceNormal}
  n3 = Constant() {r=0.5}
  n4 = CollectionParameter() {
   collection=&FortniteMaterialParameters
   parameterName=SunAndMoonModelDirectionalVector
  }
  n5 = ComponentMask(None=@n4) {b=true, g=true, r=true}
  MaterialExpressionDotProduct_3 = DotProduct(A=CameraVectorWS(), B=@n5)
  MaterialExpressionScalarParameter_3 = ScalarParameter() {defaultValue=0.4, parameterName=Anisotropy}
  n6 = Custom!(CosTheta=@MaterialExpressionDotProduct_3, k=@MaterialExpressionScalarParameter_3) {
   code="\tfloat Inner = (1 + k * CosTheta);\r\n\treturn (1 - k * k) / (4 * PI * Inner * Inner);\r\n\r\n"
   description=Schlick Phase
   inputs [
    {
     inputName=CosTheta
     input {expression=&self:MaterialExpressionDotProduct_3}
    }
    {
     inputName=k
     input {expression=&self:MaterialExpressionScalarParameter_3}
    }
   ]
  }
  n7 = Multiply(
   A=Fresnel(
    ExponentIn=ScalarParameter() {defaultValue=0.7, parameterName=LightThroughSmokeFresnelXP}
    BaseReflectFractionIn=ScalarParameter() {defaultValue=0.02, parameterName=LightThroughSmokeFresnelBase}
    Normal=MaterialFunctionCall(Normal=@n2, Flatness=@n3) {materialFunction=&FlattenNormal}.Result
   )
   B=@n6
  )
  n8 = CollectionParameter() {
   collection=&FortniteMaterialParameters
   parameterName=FogDirectionalInscatteringColor
  }
  n9 = ComponentMask(None=CollectionParameter() {collection=&FortniteMaterialParameters, parameterName=SunLightColor}) {b=true, g=true, r=true}
  n10 = Add(A=ComponentMask(None=@n8) {b=true, g=true, r=true}, B=Multiply(A=@n9) {constB=0.6}) {constB=0.1}
  FunctionOutput(None=Multiply(
    A=FunctionInput(Preview=@n1) {bUsePreviewValueAsDefault=true, inputName=ScatteringBrightness, inputType=FunctionInput_Scalar, sortPriority=4}
    B=Multiply(A=@n7, B=@n10)
   ) {constB=0.5}) {outputName=Light Through Smoke Color, sortPriority=1}
  n11 = FunctionInput(Preview=VectorParameter() {parameterName=HighlightMult, defaultValue=<rgba 0.5 0.5 0.5 0>}.RGB) {bUsePreviewValueAsDefault=true, inputName=HighlightMult, sortPriority=1}
  n12 = Add(A=DotProduct(A=Constant3Vector() {constant=<rgba 0 0 1 0>}, B=@n5)) {constB=0.2}
  n13 = DotProduct(A=@n5, B=@n2)
  n14 = FunctionInput(Preview=ScalarParameter() {defaultValue=1.5, parameterName=HighlightPower}) {bUsePreviewValueAsDefault=true, inputName=HighlightPower, inputType=FunctionInput_Scalar, sortPriority=4}
  n15 = Multiply(
   A=Multiply(A=Saturate(None=@n12), B=@n9)
   B=LinearInterpolate(A=ScalarParameter() {defaultValue=0.05, parameterName=MinHighlightBrightness}, Alpha=Power(Base=@n13, Exp=@n14)) {constA=0.07}
  )
  n16 = Add(
   A=ComponentMask(None=CollectionParameter() {collection=&FortniteMaterialParameters, parameterName=FogColor}) {b=true, g=true, r=true}
   B=VectorParameter() {parameterName=ShadowAdd, defaultValue=<rgba 0.0649 0.1616 0.2 0>}.RGB
  )
  n17 = FunctionInput(Preview=VectorParameter() {parameterName=ShadowMult, defaultValue=<rgba 0.2 0.2 0.2 0>}.RGB) {bUsePreviewValueAsDefault=true, inputName=ShadowMult, sortPriority=2}
  FunctionOutput(None=Add(A=Multiply(A=@n11, B=@n15), B=Multiply(A=@n16, B=@n17))) {outputName=Light and Shadow Colors}
 ]
}
