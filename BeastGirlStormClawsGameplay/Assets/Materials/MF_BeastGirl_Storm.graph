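# Material function that covers Beast Girl's storm-claw material in animated, emissive storm clouds with lightning, also outputting a cloud mask.
# View-facing falloff minus distorted panning noise shapes the clouds, sampled by UV or pre-skinned-position projection via static switch.
# Custom primitive data C_BeastFormEnd shifts cloud colors from purple to red.
# Emissive adds noise-timed lightning, an FOV-compensated screen-aligned glow around HeadSocketLocation, and, under UseArmSpline, wavy rippling strands.
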MaterialFunction /BeastGirlStormClawsGameplay/Assets/Materials/MF_BeastGirl_Storm
graph {nodes [
  ComponentMask(None=TextureSampleParameter2D() {parameterName=HC_DiffuseWithMask, texture=null}.RGBA) {a=true}
  n1 = CollectionParameter() {
   collection=&FortniteMaterialParameters
   parameterName=SunAndMoonModelDirectionalVector
  }
  ComponentMask(None=@n1) {b=true, g=true, r=true}
  n2 = VertexInterpolator(VS=PreSkinnedPosition())
  n3 = Add(
   A=Divide(A=@n2.PS, B=VectorParameter() {parameterName=HC_TP_CoreFlameDistortionNoise_Tiling_V2, defaultValue=<rgba 4 3 2 1>}.RGB)
   B=Multiply(A=Time(), B=VectorParameter() {parameterName=HC_TP_CoreFlame_PanSpeedAndDirection, defaultValue=<rgba -0.1 0.1 1 1>}.RGB)
  )
  n4 = TextureSampleParameter2D(UVs=ComponentMask(None=@n3) {b=true, g=true}) {parameterName=DistortTexture, samplerType=SAMPLERTYPE_LinearColor, texture=&T_NoisePack_LinearColor}
  n5 = TextureSampleParameter2D(UVs=ComponentMask(None=@n3) {b=true, r=true}) {parameterName=DistortTexture, samplerType=SAMPLERTYPE_LinearColor, texture=&T_NoisePack_LinearColor}
  n6 = PreSkinnedNormal()
  n7 = Abs(None=ComponentMask(None=VertexInterpolator(VS=@n6).PS) {g=true})
  n8 = LinearInterpolate(A=@n4.RGB, B=@n5.RGB, Alpha=@n7)
  n9 = Subtract(A=Multiply(A=@n8) {constB=2})
  ComponentMask(None=@n9) {g=true, r=true}
  Constant() {r=2}
  n10 = GetMaterialAttributes(None=FunctionInput() {inputName=Maerial, inputType=FunctionInput_MaterialAttributes}) {attributeGetTypes [
    #1
    #2
    #3
    #4
   ]}
  n11 = Constant()
  n12 = VectorParameter() {parameterName=HC_TP_StormNoise_Tiling, defaultValue=<rgba 4 4 4 1>}
  n13 = ScalarParameter() {defaultValue=0.1, parameterName=HC_TP_StormNoiseDistortionStrength_V2}
  n14 = Add(
   A=Add(A=Divide(A=@n2.PS, B=@n12.RGB), B=Multiply(A=@n13, B=@n9))
   B=Multiply(A=Time(), B=VectorParameter() {parameterName=HC_TP_StormNoise_PanDirectionAndSpeed, defaultValue=<rgba 0.025 -0.05 1 1>}.RGB)
  )
  n15 = ComponentMask(None=@n14) {b=true, g=true}
  n16 = TextureSampleParameter2D(UVs=@n15) {
   parameterName=FlameNoise_02
   samplerType=SAMPLERTYPE_LinearGrayscale
   texture=/Game/Characters/Player/Male/Medium/Bodies/M_MED_Hurricane/Textures/T_MED_Hurricane_StormCloud
  }
  n17 = ComponentMask(None=@n14) {b=true, r=true}
  n18 = TextureSampleParameter2D(UVs=@n17) {
   parameterName=FlameNoise_02
   samplerType=SAMPLERTYPE_LinearGrayscale
   texture=&T_MED_Hurricane_StormCloud
  }
  n19 = LinearInterpolate(A=@n16.RGB, B=@n18.RGB, Alpha=@n7)
  n20 = TextureCoordinate()
  n21 = ComponentMask(None=VectorParameter() {parameterName=HC_StormNoise_Tiling, defaultValue=<rgba 4 4 4 1>}.RGB) {g=true, r=true}
  n22 = ScalarParameter() {defaultValue=0.1, parameterName=HC_StormNoiseDistortionStrength}
  n23 = ComponentMask(None=VectorParameter() {parameterName=HC_CoreFlameDistortionNoise_Tiling, defaultValue=<rgba 4 3 2 1>}.RGB) {g=true, r=true}
  n24 = VectorParameter() {parameterName=HC_CoreFlame_PanSpeedAndDirection, defaultValue=<rgba -0.1 0.1 1 1>}
  n25 = TextureSampleParameter2D(UVs=Panner(Coordinate=Multiply(A=@n20, B=@n23), Speed=ComponentMask(None=@n24.RGB) {g=true, r=true})) {parameterName=DistortTexture, samplerType=SAMPLERTYPE_LinearColor, texture=&T_NoisePack_LinearColor}
  n26 = ComponentMask(None=Subtract(A=Multiply(A=@n25.RGB) {constB=2})) {g=true, r=true}
  n27 = VectorParameter() {parameterName=HC_StormNoise_PanDirectionAndSpeed, defaultValue=<rgba 0.025 -0.05 1 1>}
  n28 = Add(A=Add(A=Multiply(A=@n20, B=@n21), B=Multiply(A=@n22, B=@n26)), B=Multiply(A=Time(), B=ComponentMask(None=@n27.RGB) {g=true, r=true}))
  n29 = TextureSample(UVs=@n28) {samplerType=SAMPLERTYPE_Normal, texture=/Game/Characters/Player/Male/Medium/Bodies/M_MED_Hurricane/Textures/T_MED_Hurricane_StormCloud_N}
  n30 = Transform(None=@n29.RGB)
  n31 = OneMinus(None=DotProduct(A=CameraVectorWS(), B=@n30))
  n32 = StaticSwitchParameter(True=ComponentMask(None=@n19) {r=true}, False=Power(Base=ConstantBiasScale(None=@n31))) {parameterName=UseWorldAlignedStorm}
  n33 = ScalarParameter() {bUseCustomPrimitiveData=true, parameterName=C_BeastFormEnd}
  n34 = VectorParameter() {parameterName=RedGLow_Ending, defaultValue=<rgba 1 0 0 1>}
  n35 = DotProduct(A=CameraVectorWS(), B=Transform(None=@n10.Normal))
  n36 = Constant()
  n37 = StaticSwitchParameter(True=ComponentMask(None=@n8) {r=true}, False=@n25.R) {parameterName=UseWorldAlignedStorm}
  n38 = LinearInterpolate(
   A=ScalarParameter() {defaultValue=1.5, parameterName=HC_StormCoverageMax}
   B=ScalarParameter() {defaultValue=1, parameterName=HC_StormCoverageMin}
   Alpha=@n37
  )
  n39 = ScalarParameter() {defaultValue=0.6, parameterName=HC_StormHardness}
  n40 = TextureSampleParameter2D(UVs=@n28) {
   parameterName=FlameNoise_02
   samplerType=SAMPLERTYPE_LinearGrayscale
   texture=&T_MED_Hurricane_StormCloud
  }
  n41 = StaticSwitchParameter(True=@n19, False=@n40.R) {parameterName=UseWorldAlignedStorm}
  n42 = LinearInterpolate(
   A=ScalarParameter() {defaultValue=0.5, parameterName=HC_StormNoise_SubtractionStrength_Max}
   B=ScalarParameter() {defaultValue=0.25, parameterName=HC_StormNoise_SubtractionStrength_Min}
   Alpha=@n37
  )
  n43 = Clamp(None=Subtract(A=SphereMask(A=@n35, B=@n36, Radius=@n38, Hardness=@n39), B=Multiply(A=@n41, B=@n42)))
  n44 = LinearInterpolate(
   A=LinearInterpolate(
    A=VectorParameter() {parameterName=HC_StormColor_Dark, defaultValue=<rgba 0.0307 0.008 0.0999 1>}.RGB
    B=VectorParameter() {parameterName=RedGLow_Ending2, defaultValue=<rgba 0.1649 0 0 1>}.RGB
    Alpha=@n33
   )
   B=LinearInterpolate(A=VectorParameter() {parameterName=HC_StormColor_Light, defaultValue=<rgba 0.1248 0.0497 0.7011 1>}.RGB, B=@n34.RGB, Alpha=@n33)
   Alpha=Multiply(A=@n43, B=@n43) {constB=4}
  )
  n45 = Clamp(None=Multiply(A=@n43) {constB=2})
  n46 = Multiply(A=Multiply(A=Multiply(A=@n32, B=@n44), B=@n45), B=Multiply(A=@n41, B=ScalarParameter() {defaultValue=2, parameterName=HC_ShadowStrength}))
  n47 = Constant() {r=1}
  n48 = OneMinus(None=@n11)
  n49 = TextureCoordinate()
  n50 = ScalarParameter() {defaultValue=1, group=Lightning/Thunder Parameters, parameterName=" LightningTiling"}
  n51 = Time()
  n52 = Add(A=@n51) {constB=0.35}
  n53 = Add(A=@n51) {constB=0.75}
  n54 = VectorNoise(World Position=Multiply(
    A=MaterialFunctionCall(X=@n51, Y=@n52, Z=@n53) {materialFunction=&MakeFloat3}.Result
    B=ScalarParameter() {defaultValue=1, group=Lightning/Thunder Parameters, parameterName=LightningFlashSpeed}
   ) {constB=12})
  n55 = TextureSample(UVs=Add(A=Multiply(A=@n49, B=@n50), B=ComponentMask(None=@n54) {g=true})) {texture=/Game/Global/Textures/Default/Noise/T_Noise_Cloud}
  n56 = ComponentMask(None=@n54) {r=true}
  n57 = Ceil(None=Multiply(A=Clamp(None=@n56)) {constB=2})
  n58 = Divide(A=@n57)
  n59 = Time()
  n60 = ScalarParameter() {defaultValue=1, group=Lightning/Thunder Parameters, parameterName=LightningSpeed}
  n61 = Subtract(
   A=LinearInterpolate(A=LinearInterpolate(A=@n55.R, B=@n55.R, Alpha=@n57), B=@n55.B, Alpha=Floor(None=@n58))
   B=Frac(None=Multiply(A=@n59, B=@n60) {constB=0.2})
  ) {constB=0.2}
  n62 = ScalarParameter() {defaultValue=0.1, group=Lightning/Thunder Parameters, parameterName=LightningCoveraeFalloff}
  n63 = Power(Base=Clamp(None=Subtract(A=@n61, B=@n62))) {constExponent=4}
  n64 = ScalarParameter() {defaultValue=2, group=Lightning/Thunder Parameters, parameterName=ThunderTiling}
  n65 = TextureCoordinate()
  n66 = Multiply(A=@n65, B=Multiply(A=@n64, B=ScalarParameter() {defaultValue=4, group=Lightning/Thunder Parameters, parameterName=ThunderArcNoise}))
  n67 = Multiply(A=ComponentMask(None=TextureSample(UVs=@n66) {texture=&T_Noise_Cloud}.RGB) {g=true, r=true}) {constB=2}
  n68 = Add(
   A=Multiply(A=TextureCoordinate(), B=@n64)
   B=Multiply(
    A=Subtract(A=@n67)
    B=ScalarParameter() {defaultValue=0.025, group=Lightning/Thunder Parameters, parameterName=ThunderArcDistortion strength}
   )
  )
  n69 = Multiply(A=@n64, B=ScalarParameter() {defaultValue=1, group=Lightning/Thunder Parameters, parameterName=ThunderDistortNoise})
  n70 = Multiply(A=@n65, B=@n69)
  n71 = Multiply(A=TextureSample(UVs=@n69) {texture=&T_Noise_Cloud}.RGB) {constB=2}
  n72 = Multiply(A=ComponentMask(None=Subtract(A=@n71)) {g=true, r=true}, B=Divide(A=@n69))
  n73 = ComponentMask(None=TextureSample(UVs=Panner(Coordinate=@n70, Speed=@n72) {speedX=0.5, speedY=0.5}) {texture=&T_Noise_Cloud}.RGB) {g=true, r=true}
  n74 = Multiply(
   A=Subtract(A=Multiply(A=@n73) {constB=2})
   B=ScalarParameter() {defaultValue=0.05, group=Lightning/Thunder Parameters, parameterName=ThunderMAcroDistortion strength}
  )
  n75 = ComponentMask(None=TextureSample(UVs=Add(A=@n68, B=@n74)) {samplerType=SAMPLERTYPE_Masks, texture=/Game/Weapons/WeaponSkins/Wraps/Textures/PlasmaCharge/PlasmaCharge_M}.RGB) {g=true}
  n76 = ScalarParameter() {defaultValue=15, group=Lightning/Thunder Parameters, parameterName=ThunderBrightness}
  n77 = Multiply(A=@n63, B=Power(Base=Multiply(A=@n75, B=@n76) {constB=10}))
  n78 = Multiply(
   A=ComponentMask(None=VectorParameter() {group=Lightning/Thunder Parameters, parameterName=LightningColor, defaultValue=<rgba 0.6495 0 1 1>}.RGB) {b=true, g=true, r=true}
   B=Multiply(A=Add(A=@n63, B=@n77), B=ScalarParameter() {defaultValue=1, group=Lightning/Thunder Parameters, parameterName=bEnableThunderlightning})
  )
  n79 = ScalarParameter() {defaultValue=10, group=Lightning/Thunder Parameters, parameterName=LightningBrightness}
  n80 = LinearInterpolate(A=@n79, B=Multiply(A=@n79) {constB=2}, Alpha=Clamp(None=ComponentMask(None=@n54) {g=true}))
  n81 = Multiply(A=@n46, B=ScalarParameter() {defaultValue=2, parameterName=HC_Clouds_EmissiveStrength})
  n82 = VectorParameter() {parameterName=HC_CloudEdgeGlow_Color, defaultValue=<rgba 1 0.3012 0 1>}
  n83 = MaterialFunctionCall!(Center Point World Position=VectorParameter() {parameterName=HeadSocketLocation, defaultValue=<rgba 0 0 0 1>}.RGB) {materialFunction=/Engine/Functions/Engine_MaterialFunctions02/UVs/WorldSpaceAlignedScreenCoordinates}
  n84 = ScalarParameter() {defaultValue=12, parameterName=UV_Scale}
  n85 = MaterialFunctionCall() {materialFunction=/Engine/Functions/Engine_MaterialFunctions02/Utility/FOV}
  n86 = Constant() {r=0.414}
  n87 = Constant() {r=1.091}
  n88 = Constant()
  n89 = Constant() {r=1}
  n90 = LinearInterpolate(
   A=Multiply(A=@n84) {constB=2}
   B=@n84
   Alpha=Saturate(None=MaterialFunctionCall(Input=@n85."FOV ", Input Low=@n86, Input High=@n87, Target Low=@n88, Target High=@n89) {materialFunction=&RemapValueRange}.Result)
  )
  n91 = Subtract(A=Multiply(A=MaterialFunctionCall(UVs=@n83.Offset UVs, Texture Scale=@n90) {materialFunction=&ScaleUVsByCenter}.UVs) {constB=2})
  n92 = Constant2Vector()
  n93 = ScalarParameter() {defaultValue=2, parameterName=GlowMaxRadius}
  n94 = Constant()
  n95 = Clamp(None=Power(Base=SphereMask(A=@n91, B=@n92, Radius=@n93, Hardness=@n94)))
  n96 = Multiply(
   A=Power(Base=@n32)
   B=Multiply(A=Multiply(A=@n82.RGB, B=@n95), B=ScalarParameter() {defaultValue=2.5, parameterName=HC_CloudEdgeGlowStrength})
  )
  n97 = Add(A=Power(Base=Multiply(A=@n78, B=@n80) {constB=4}), B=LinearInterpolate(B=Add(A=@n81, B=@n96), Alpha=@n32))
  n98 = TextureCoordinate() {coordinateIndex=1}
  n99 = Constant()
  n100 = ScalarParameter() {defaultValue=0.25, parameterName=HC_HeightRatio}
  n101 = Subtract(A=Multiply(A=BumpOffset(Coordinate=@n98, Height=@n99, HeightRatioInput=@n100) {heightRatio=0}) {constB=2})
  n102 = TextureCoordinate() {coordinateIndex=1}
  n103 = ScalarParameter() {defaultValue=1, parameterName=HC_Strand_Noise_Tiling}
  n104 = TextureSample(UVs=Panner(Coordinate=Multiply(A=@n102, B=@n103)) {speedY=0.2}) {samplerType=SAMPLERTYPE_Masks, texture=/Game/Global/Textures/Default/Noise/T_WrapEdgeWear_Noise}
  n105 = Add(A=Time(), B=LinearInterpolate(A=Constant(), B=Constant() {r=0.75}, Alpha=@n104.R))
  n106 = Multiply(
   A=Add(
    A=MaterialFunctionCall(Float2=@n101) {materialFunction=&BreakOutFloat2Components}.G
    B=Multiply(A=@n105) {constB=0.25}
   )
   B=ScalarParameter() {defaultValue=1, parameterName=HC_WaveFrequency}
  ) {constB=5}
  n107 = ScalarParameter() {defaultValue=0.025, parameterName=HC_Amplitude_Min}
  n108 = ScalarParameter() {defaultValue=0.25, parameterName=HC_Amplitude_Max}
  n109 = Subtract(A=ComponentMask(None=@n101) {r=true}, B=Multiply(A=Sine(None=@n106), B=LinearInterpolate(A=@n107, B=@n108, Alpha=@n104.R)) {constB=0.1})
  n110 = Add(A=@n109, B=ScalarParameter() {defaultValue=0.1, parameterName=HC_Strand_02_offset}) {constB=-0.75, desc=1}
  n111 = Constant()
  n112 = LinearInterpolate(
   A=ScalarParameter() {defaultValue=0.02, parameterName=HC_StrandThickness}
   B=ScalarParameter() {defaultValue=0.005, parameterName=HC_StrandThickness_Min}
   Alpha=@n104.R
  )
  n113 = Constant()
  n114 = Add(A=@n109, B=ScalarParameter() {defaultValue=-0.25, parameterName=HC_StrandOffset})
  n115 = Max(A=SphereMask(A=@n110, B=@n111, Radius=@n112, Hardness=@n113), B=SphereMask(A=@n114, B=@n111, Radius=@n112, Hardness=@n113))
  n116 = ConstantBiasScale(None=Sine(None=Multiply(A=@n115) {constB=2}))
  n117 = Multiply(
   A=Multiply(A=VectorParameter() {parameterName=HC_Strand_GlowColor, defaultValue=<rgba 1 0.1992 0 1>}.RGB, B=Multiply(A=@n116, B=@n115))
   B=ScalarParameter() {defaultValue=25, parameterName=Strand_GlowStrength}
  )
  n118 = ScalarParameter() {defaultValue=0.1, parameterName=HC_StrandRipple_raadius}
  n119 = Constant()
  n120 = Max(A=SphereMask(A=@n110, B=@n111, Radius=@n118, Hardness=@n119), B=SphereMask(A=@n114, B=@n111, Radius=@n118, Hardness=@n119))
  n121 = Sine(None=Time())
  n122 = Multiply(
   A=Multiply(A=VectorParameter() {parameterName=HC_StrandGlowColor, defaultValue=<rgba 1 0.5161 0 1>}.RGB, B=Divide(A=@n120) {constB=48})
   B=Multiply(A=ScalarParameter() {defaultValue=2, parameterName=HC_StrandGlowStength}, B=ConstantBiasScale(None=@n121))
  )
  n123 = Divide(A=@n120)
  n124 = Multiply(A=Time(), B=ScalarParameter() {defaultValue=1, parameterName=HC_StrandRippleSpeed})
  n125 = Frac(None=Multiply(A=Add(A=@n120, B=@n124)) {constB=7})
  n126 = Multiply(A=@n125, B=@n125)
  n127 = Multiply(A=Add(A=Time()) {constB=0.25}, B=ScalarParameter() {defaultValue=0.75, parameterName=HC_StrandRippleSpeed_02})
  n128 = Frac(None=Multiply(A=Add(A=@n120, B=@n127)) {constB=5})
  n129 = Multiply(A=@n128, B=@n128)
  n130 = Add(
   A=Multiply(A=Constant3Vector() {constant=<rgba 0 0.2303 1 0>}, B=Multiply(A=@n123, B=@n126))
   B=Multiply(A=Constant3Vector() {constant=<rgba 0.4612 0 1 0>}, B=Multiply(A=@n123, B=@n129))
  )
  n131 = ScalarParameter() {defaultValue=1, parameterName=HC_StrandRippleBirghtness}
  n132 = Add(A=@n97, B=Add(A=Max(A=@n117, B=@n122), B=Multiply(A=@n130, B=@n131)))
  n133 = SetMaterialAttributes!(
   MaterialAttributes=@n10.MaterialAttributes
   BaseColor=LinearInterpolate(A=@n46, B=@n10.BaseColor, Alpha=@n11)
   Metallic=Constant()
   Specular=LinearInterpolate(A=Constant() {r=1}, B=@n10.Specular, Alpha=@n11)
   Normal=@n10.MaterialAttributes
   EmissiveColor=StaticSwitchParameter(True=Multiply(A=@n48, B=@n132), False=Multiply(A=@n48, B=@n97)) {parameterName=UseArmSpline}
   Roughness=LinearInterpolate(A=@n47, B=@n10.Roughness, Alpha=@n11)
  ) {attributeSetTypes [
    #2
    #5
    #4
    #1
    #6
    #3
   ]}
  FunctionOutput(None=@n133)
  FunctionOutput(None=OneMinus(None=@n45)) {outputName=CloudMask}
  FunctionOutput(None=@n11) {outputName=DirtMask}
  n134 = TextureSample(UVs=@n15) {samplerType=SAMPLERTYPE_Normal, texture=&T_MED_Hurricane_StormCloud_N}
  n135 = TextureSample(UVs=@n17) {samplerType=SAMPLERTYPE_Normal, texture=&T_MED_Hurricane_StormCloud_N}
  LinearInterpolate(A=@n134.RGB, B=@n135.RGB, Alpha=@n7)
  LinearInterpolate(B=Multiply(A=Multiply(B=@n34.RGB), B=ScalarParameter() {defaultValue=5, parameterName=Glow_Strength}), Alpha=@n33)
  n136 = Time()
  LinearInterpolate(
   A=ScalarParameter() {defaultValue=1, parameterName=GlowMinRadius}
   B=@n93
   Alpha=MaterialFunctionCall(Float=Multiply(A=@n136) {constB=10}) {materialFunction=&MF_Hash}.Float
  )
  MaterialFunctionCall!(WorldSpaceNormal, HighlightMult, ShadowMult, ScatteringBrightness, HighlightPower) {materialFunction=/BRCosmetics/Effects/Fort_Effects/Materials/_Functions/MF_Unlit_Highlight_And_Shadow}
  n137 = VectorParameter() {parameterName=HC_StormColor_Highlight, defaultValue=<rgba 0.7011 0.7011 0.7011 1>}
  n138 = VectorParameter() {parameterName=HC_StormColor_Shadow, defaultValue=<rgba 0.0999 0.0999 0.0999 1>}
  n139 = ScalarParameter() {defaultValue=0.5, parameterName=HC_Scattering_Mult}
  n140 = ScalarParameter() {defaultValue=0.5, parameterName=HC_HiglightPower}
  Multiply(
   A=Add(A=MaterialFunctionCall(WorldSpaceNormal=@n30, HighlightMult=@n137.RGB, ShadowMult=@n138.RGB, ScatteringBrightness=@n139, HighlightPower=@n140) {materialFunction=&MF_Unlit_Highlight_And_Shadow}.Light and Shadow Colors) {constB=0}
   B=@n44
  )
  Reroute(None=@n47)
 ]}
