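# Temporary Default Lit material for the BeastGirl StormClaws character, set up for skeletal meshes and mesh deformers.
# A Normal texture parameter, defaulting to the Vertigo large male body normal map, is written into material attributes as the only attribute set.
# Those attributes feed the MF_BeastGirl_Storm material function, leaving any storm look to that function.
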Material /BeastGirlStormClawsGameplay/Assets/Materials/M_BeastGirlTemp
graph {
 properties {
  bAutomaticallySetUsageInEditor=false
  bUseMaterialAttributes=true
  bUsedWithMeshDeformer=true
  bUsedWithSkeletalMesh=true
  usedShadingModels=Default Lit
 }
 nodes [
  n1 = TextureSampleParameter2D() {
   parameterName=Normal
   samplerType=SAMPLERTYPE_Normal
   texture=/Game/Characters/Player/Male/Large/Bodies/M_LRG_Vertigo/Textures/T_M_LRG_Vertigo_Body_N
  }
  MaterialFunctionCall!(Maerial=SetMaterialAttributes!(MaterialAttributes, Normal=@n1.RGB) {attributeSetTypes [#1]}) {materialFunction=&/MF_BeastGirl_Storm}
 ]
}
